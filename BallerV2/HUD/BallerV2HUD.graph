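# Vehicle dashboard HUD widget for the BallerV2, showing an energy bar with boost effects and low or depleted energy warnings.
# Fuel change events cache the value, driving the bar material's progress, spark position, and glow size and fade.
# Boost events play glow and spark fade animations, while low and empty states trigger animated warning messages.
# Infinite fuel hides the bar row, and Construct checks for touch input around repositioning it.
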WidgetBlueprint /BallerV2/HUD/BallerV2HUD
animations [
 Anim_BoostOn {
  binding_count=3
  length_seconds=0.2
  object=&self:Anim_BoostOn
  sequence {bindings [
    FillBar {
     id=#1
     possessed_class=Image
     tracks [{
       type=MovieSceneWidgetMaterialTrack
       sections [{
         type=MovieSceneParameterSection
         channels [
          Progress_Color_StopR {type=MovieSceneScriptingFloatChannel, const [1,2]}
          Progress_Color_StopG {type=MovieSceneScriptingFloatChannel, const [1,2]}
          Progress_Color_StopB {type=MovieSceneScriptingFloatChannel, const [1,2]}
          Progress_Color_StopA {type=MovieSceneScriptingFloatChannel, const [1,2]}
         ]
        }]
      }]
    }
    BoostGlowBorder {
     id=#2
     possessed_class=Border
     tracks [{
       type=MovieSceneFloatTrack
       display_name=Render Opacity
       sections [{type=MovieSceneFloatSection, channels [None {type=MovieSceneScriptingFloatChannel, keys [[0,0],[4,1]]}]}]
      }]
    }
    Overlay_BoostSparks {
     id=#3
     possessed_class=Overlay
     tracks [{
       type=MovieSceneFloatTrack
       display_name=Render Opacity
       sections [{type=MovieSceneFloatSection, channels [None {type=MovieSceneScriptingFloatChannel, keys [[0,0],[4,1]]}]}]
      }]
    }
   ]}
 }
 Anim_BoostOff {
  binding_count=3
  length_seconds=0.75
  object=&self:Anim_BoostOff
  sequence {bindings [
    Overlay_BoostSparks {
     id=#4
     possessed_class=Overlay
     tracks [{
       type=MovieSceneFloatTrack
       display_name=Render Opacity
       sections [{type=MovieSceneFloatSection, channels [None {type=MovieSceneScriptingFloatChannel, keys [[0,1],[14.9997,0]]}]}]
      }]
    }
    FillBar {
     id=#1
     possessed_class=Image
     tracks [{
       type=MovieSceneWidgetMaterialTrack
       sections [{
         type=MovieSceneParameterSection
         channels [
          Progress_Color_StopR {type=MovieSceneScriptingFloatChannel, const [1,2]}
          Progress_Color_StopG {type=MovieSceneScriptingFloatChannel, const [1,2]}
          Progress_Color_StopB {type=MovieSceneScriptingFloatChannel, const [1,2]}
          Progress_Color_StopA {type=MovieSceneScriptingFloatChannel, const [1,2]}
         ]
        }]
      }]
    }
    BoostGlowBorder {
     id=#5
     possessed_class=Border
     tracks [{
       type=MovieSceneFloatTrack
       display_name=Render Opacity
       sections [{type=MovieSceneFloatSection, channels [None {type=MovieSceneScriptingFloatChannel, keys [[0,1],[15,0]]}]}]
      }]
    }
   ]}
 }
 Anim_LowFuel_Messaging {
  binding_count=6
  length_seconds=0.25
  object=&self:Anim_LowFuel_Messaging
  sequence {
   bindings [
    TriangleArrow {
     id=#6
     possessed_class=Image
     tracks [
      {
       type=MovieSceneFloatTrack
       display_name=Render Opacity
       sections [{type=MovieSceneFloatSection, channels [None {type=MovieSceneScriptingFloatChannel, keys [[0.03,0],[3,1]]}]}]
      }
      {
       type=MovieScene2DTransformTrack
       sections [{
         type=MovieScene2DTransformSection
         channels [
          Translation.X {type=MovieSceneScriptingFloatChannel, keys [[-0.001,16],[5.999,-12],[8.999,-8]]}
          Translation.Y {type=MovieSceneScriptingFloatChannel, keys [[-0.001,0],[5.999,30],[8.999,24]]}
          Angle {type=MovieSceneScriptingFloatChannel, const [108,1]}
          Scale.X {type=MovieSceneScriptingFloatChannel, const [1,1]}
          Scale.Y {type=MovieSceneScriptingFloatChannel, const [1,1]}
          Shear.X {type=MovieSceneScriptingFloatChannel, const [0,1]}
          Shear.Y {type=MovieSceneScriptingFloatChannel, const [0,1]}
         ]
        }]
      }
     ]
    }
    Image_NoEnergy_GlowPulse {
     id=#7
     possessed_class=Image
     tracks [{
       type=MovieSceneFloatTrack
       display_name=Render Opacity
       sections [{type=MovieSceneFloatSection, channels [None {type=MovieSceneScriptingFloatChannel, keys [[0,0],[12,0.5]]}]}]
      }]
    }
    AdditonalOuterRing {
     id=#8
     possessed_class=Image
     tracks [
      {
       type=MovieSceneFloatTrack
       display_name=Render Opacity
       sections [{type=MovieSceneFloatSection, channels [None {type=MovieSceneScriptingFloatChannel, keys [[9,0],[12.001,0.25]]}]}]
      }
      {
       type=MovieScene2DTransformTrack
       sections [{
         type=MovieScene2DTransformSection
         channels [
          Scale.X {type=MovieSceneScriptingFloatChannel, keys [[9,2],[15,1]]}
          Scale.Y {type=MovieSceneScriptingFloatChannel, keys [[9,2],[15,1]]}
         ]
        }]
      }
     ]
    }
    WarningMessaging {
     id=#9
     possessed_class=Overlay
     tracks [{
       type=MovieSceneFloatTrack
       display_name=Render Opacity
       sections [{type=MovieSceneFloatSection, channels [None {type=MovieSceneScriptingFloatChannel, keys [[0,0],[3,1]]}]}]
      }]
    }
    TextContainerLow {
     id=#10
     possessed_class=Overlay
     tracks [
      {
       type=MovieSceneFloatTrack
       display_name=Render Opacity
       sections [{type=MovieSceneFloatSection, channels [None {type=MovieSceneScriptingFloatChannel, keys [[6,0],[9,1]]}]}]
      }
      {
       type=MovieScene2DTransformTrack
       sections [{
         type=MovieScene2DTransformSection
         channels [
          Translation.X {type=MovieSceneScriptingFloatChannel, keys [[6,20],[9,0],[12.001,8]]}
          Translation.Y {type=MovieSceneScriptingFloatChannel, keys [[6,-16],[9,0],[12.001,-8]]}
          Scale.X {type=MovieSceneScriptingFloatChannel, keys [[6,1.125],[9,0.95],[12.001,1]]}
          Scale.Y {type=MovieSceneScriptingFloatChannel, keys [[6,1.125],[9,0.95],[12.001,1]]}
         ]
        }]
      }
     ]
    }
    Image_Energy {
     id=#11
     possessed_class=Image
     tracks [{
       type=MovieScene2DTransformTrack
       sections [{
         type=MovieScene2DTransformSection
         channels [
          Translation.X {type=MovieSceneScriptingFloatChannel, const [0,1]}
          Translation.Y {type=MovieSceneScriptingFloatChannel, keys [[0,-1],[5.999,-1],[9,-4],[12,-1]]}
          Scale.X {type=MovieSceneScriptingFloatChannel, keys [[0,1],[5.999,1],[9,1.125],[12,1]]}
          Scale.Y {type=MovieSceneScriptingFloatChannel, keys [[0,1],[5.999,1],[9,1.125],[12,1]]}
         ]
        }]
      }]
    }
   ]
   display_rate {numerator=60}
  }
 }
 Anim_SelfDestructMessaging {
  binding_count=5
  length_seconds=0.35
  object=&self:Anim_SelfDestructMessaging
  sequence {
   bindings [
    TextContainerEject {
     id=#12
     possessed_class=Overlay
     tracks [
      {
       type=MovieSceneFloatTrack
       display_name=Render Opacity
       sections [{type=MovieSceneFloatSection, channels [None {type=MovieSceneScriptingFloatChannel, keys [[6,0],[15,1]]}]}]
      }
      {
       type=MovieScene2DTransformTrack
       sections [{
         type=MovieScene2DTransformSection
         channels [
          Translation.X {type=MovieSceneScriptingFloatChannel, keys [[6,256],[15,8]]}
          Translation.Y {type=MovieSceneScriptingFloatChannel, const [-8,1]}
          Scale.X {type=MovieSceneScriptingFloatChannel, const [1,1]}
          Scale.Y {type=MovieSceneScriptingFloatChannel, const [1,1]}
         ]
        }]
      }
     ]
    }
    AlertIcon {
     id=#13
     possessed_class=Overlay
     tracks [
      {
       type=MovieSceneFloatTrack
       display_name=Render Opacity
       sections [{type=MovieSceneFloatSection, channels [None {type=MovieSceneScriptingFloatChannel, keys [[0,0],[12,0],[15,1]]}]}]
      }
      {
       type=MovieScene2DTransformTrack
       sections [{
         type=MovieScene2DTransformSection
         channels [
          Translation.X {type=MovieSceneScriptingFloatChannel, const [-32,1]}
          Translation.Y {type=MovieSceneScriptingFloatChannel, const [-40,1]}
          Scale.X {type=MovieSceneScriptingFloatChannel, keys [[12,3],[15,0.9],[21,1]]}
          Scale.Y {type=MovieSceneScriptingFloatChannel, keys [[12,3],[15,0.9],[21,1]]}
         ]
        }]
      }
     ]
    }
    AdditonalOuterRing_1 {id=#14, possessed_class=Image, tracks []}
    TextContainerLow {
     id=#10
     possessed_class=Overlay
     tracks [
      {
       type=MovieSceneFloatTrack
       display_name=Render Opacity
       sections [{type=MovieSceneFloatSection, channels [None {type=MovieSceneScriptingFloatChannel, keys [[0,1],[9,0]]}]}]
      }
      {
       type=MovieScene2DTransformTrack
       sections [{
         type=MovieScene2DTransformSection
         channels [
          Translation.X {type=MovieSceneScriptingFloatChannel, keys [[0,8],[9,-256]]}
          Translation.Y {type=MovieSceneScriptingFloatChannel, const [-8,1]}
          Scale.X {type=MovieSceneScriptingFloatChannel, const [1,1]}
          Scale.Y {type=MovieSceneScriptingFloatChannel, const [1,1]}
         ]
        }]
      }
     ]
    }
   ]
   display_rate {numerator=60}
  }
 }
]
graphs [
 EventGraph {nodes [
   n1 = Event PreConstruct()
   > Branch(Condition=@n1.IsDesignTime)
   > UpdateLowFuelState()
   > UpdateEmptyState()
   Event OnBoostStarted() {is=title K2Node_Event}
   > SetBoostState!(execute, self, bIsBoosting:bool=true) -> (then)
   Event OnBoostFinished() {is=title K2Node_Event}
   > SetBoostState(bIsBoosting=false)
   n2 = Event OnFuelValueChanged!() -> (OutputDelegate, then, FuelPercentage:float, bIsLowOnFuel:bool) {is=title K2Node_Event}
   > Branch(Condition="Not Equal (Float)"(A=@n2.FuelPercentage, B=Get CachedFuel() -> (CachedFuel:double)))
   > n3 = Set CachedFuel(CachedFuel:double=@n2.FuelPercentage) -> (Output_Get:double)
   > Set IsEmpty(IsEmpty:bool="Nearly Equal (Float)"(A=@n3, B=0, ErrorTolerance=0)) -> (Output_Get:bool)
   > n4 = Branch(Condition="Equal (Boolean)"(A:bool=@n2.bIsLowOnFuel, B:bool=Get IsLowFuel() -> (IsLowFuel:bool)) -> (ReturnValue:bool=~) {is=title K2Node_PromotableOperator})
   n5 = GetDynamicMaterial(execute=[@n13.then, @n4.then], self=Get FillBar() -> (FillBar:Image Object Reference))
   > SetScalarParameterValue(self=@n5, ParameterName=Progress, Value=Get CachedFuel())
   > UpdateEmptyState()
   n6 = Get BoostSparksEndTranslation() -> (BoostSparksEndTranslation:double)
   n7 = Get BoostSparksStartTranslation() -> (BoostSparksStartTranslation:double)
   n8 = Get CachedFuel()
   > SetRenderTranslation(
    self=Get Overlay_BoostSparks() -> (Overlay_BoostSparks:Overlay Object Reference)
    Translation=MakeVector2D(X=Lerp(A=@n6, B=@n7, Alpha=@n8), Y=0)
   )
   n9 = Get CachedFuel()
   n10 = Get BoostGlowBorderScaleEnd() -> (BoostGlowBorderScaleEnd:double)
   n11 = Get BoostGlowBorderScaleStart() -> (BoostGlowBorderScaleStart:double)
   > SetDesiredSizeScale!(
    execute
    self:Border Object Reference=Get BoostGlowBorder() -> (BoostGlowBorder:Border Object Reference)
    InScale:Vector 2D Structure=MakeVector2D(X=Lerp(A=@n10, B=@n11, Alpha=@n9), Y=0)
   ) -> (then)
   > n12 = GetDynamicMaterial(self=Get BoostGlow() -> (BoostGlow:Image Object Reference))
   > SetScalarParameterValue(self=@n12, ParameterName=FadeRate, Value=Lerp(A=1, B=3, Alpha=Get CachedFuel()))
   Set IsLowFuel(execute=@n4.else, IsLowFuel:bool=@n2.bIsLowOnFuel) -> (Output_Get:bool)
   > n13 = UpdateLowFuelState()
   Event OnOutOfFuel() {is=title K2Node_Event}
   > UpdateEmptyState()
   Event OnLowFuel() {is=title K2Node_Event}
   > UpdateLowFuelState()
   Event Construct()
   > SetVisibility(
    self=[
     Get WarningMessaging() -> (WarningMessaging:Overlay Object Reference)
     Get Image_NoEnergy_GlowPulse() -> (Image_NoEnergy_GlowPulse:Image Object Reference)
     Get Overlay_BoostSparks()
     Get BoostGlowBorder()
    ]
    InVisibility=Collapsed
   )
   > Branch(Condition=IsUsingTouch(PlayerController=GetOwningPlayer()))
   n14 = SlotAsOverlaySlot(Widget=Get HorizontalBox_1() -> (HorizontalBox_1:Horizontal Box Object Reference))
   > SetHorizontalAlignment(self=@n14, InHorizontalAlignment=HAlign_Center)
   > SetPadding!(
    execute
    self:Overlay Slot Object Reference=@n14
    InPadding_Left:float=510
    InPadding_Top:float=0
    InPadding_Right:float=0
    InPadding_Bottom:float=52
   ) -> (then)
   Event OnWidgetVisibilityChanged!() -> (OutputDelegate, then, bNewVisibility:bool) {is=title K2Node_Event}
   > Set CachedFuel(CachedFuel=-1)
   n15 = Event SetVehicleHasInfiniteFuel!() -> (OutputDelegate, then, bInfiniteFuel:bool) {is=title K2Node_Event}
   > n16 = Branch(Condition=@n15.bInfiniteFuel)
   > SetVisibility(self=Get HorizontalBox_1(), InVisibility=Hidden)
   SetActiveWidgetIndex(execute=@n16.else, self=Get Switcher_Energy() -> (Switcher_Energy:Widget Switcher Object Reference), Index=0)
  ]}
 SetBoostFXCollapsed {nodes [
   SetBoostFXCollapsed() {is=title K2Node_FunctionEntry}
   > SetVisibility(self=[Get BoostGlowBorder(), Get Overlay_BoostSparks()], InVisibility=Collapsed)
  ]}
 SetBoostState {nodes [
   n1 = SetBoostState!() -> (then, bIsBoosting:bool=~) {is=title K2Node_FunctionEntry}
   > n2 = Branch(Condition=@n1)
   > PlayAnimation(InAnimation=Get Anim_BoostOn() -> (Anim_BoostOn:Widget Animation Object Reference))
   > SetVisibility(self=[Get BoostGlowBorder(), Get Overlay_BoostSparks()], InVisibility=SelfHitTestInvisible)
   n3 = Get Anim_BoostOff() -> (Anim_BoostOff:Widget Animation Object Reference)
   PlayAnimation(execute=@n2.else, InAnimation=@n3)
   > BindToAnimationFinished(Animation=@n3, Delegate=Create Event())
  ]}
 UpdateEmptyState {nodes [
   UpdateEmptyState() {is=title K2Node_FunctionEntry}
   > Branch(Condition=Get IsEmpty() -> (IsEmpty:bool))
   > PlayAnimation(InAnimation=Get Anim_SelfDestructMessaging() -> (Anim_SelfDestructMessaging:Widget Animation Object Reference))
  ]}
 UpdateLowFuelState {nodes [
   UpdateLowFuelState() {is=title K2Node_FunctionEntry}
   > SetVisibility(
    self=[Get WarningMessaging(), Get Image_NoEnergy_GlowPulse()]
    InVisibility=Select(Option 0=Collapsed, Option 1=SelfHitTestInvisible, Index=Get IsLowFuel())
   )
   > PlayAnimation(InAnimation=Get Anim_LowFuel_Messaging() -> (Anim_LowFuel_Messaging:Widget Animation Object Reference))
  ]}
]
variables [
 BoostGlowBorderScaleEnd {type=number}
 BoostGlowBorderScaleStart {type=number, default=1}
 BoostSparksEndTranslation {type=number, default=-296}
 BoostSparksStartTranslation {type=number}
 CachedFuel {type=number, default=-1}
 IsEmpty {type=boolean}
 IsLowFuel {type=boolean}
]
widgets {
 info {
  inheritedWidgetCount=10
  parentClass=/Script/FortniteUI.FortAthenaVehicleDashboardWidget
  rootWidgetClass=&Overlay
  widgetCount=51
 }
 widgets [
  Overlay_1 {widgetClassPath=&Overlay, widget=Overlay_1}
  SafeZone {
   widgetClassPath=&SafeZone
   bIsVariable=true
   parent=Overlay_1
   widget=SafeZone
   properties {padBottom=false, padLeft=false, padTop=false}
   slot {class=OverlaySlot, properties {horizontalAlignment=HAlign_Fill, verticalAlignment=VAlign_Fill}}
  }
  Overlay_5 {widgetClassPath=&Overlay, bIsVariable=true, parent=SafeZone, widget=Overlay_5, slot {class=SafeZoneSlot}}
  HorizontalBox_1 {
   widgetClassPath=&HorizontalBox
   bIsVariable=true
   parent=Overlay_5
   widget=HorizontalBox_1
   slot {class=OverlaySlot, properties {horizontalAlignment=HAlign_Right, verticalAlignment=VAlign_Bottom, padding=<ltrb 0 0 100 100>}}
  }
  Switcher_Energy {widgetClassPath=&WidgetSwitcher, bIsVariable=true, parent=HorizontalBox_1, widget=Switcher_Energy, slot {class=HorizontalBoxSlot}}
  Overlay_2 {widgetClassPath=&Overlay, parent=Switcher_Energy, widget=Overlay_2, slot {class=WidgetSwitcherSlot}}
  Image_NoEnergy_GlowPulse {
   widgetClassPath=&Image
   bIsVariable=true
   parent=Overlay_2
   widget=Image_NoEnergy_GlowPulse
   properties {
    renderOpacity=0.5
    visibility=SelfHitTestInvisible
    brush {
     bIsDynamicallyLoaded=false
     drawAs=Image
     imageType=NoImage
     mirroring=NoMirror
     resourceName=None
     resourceObject=/Game/Athena/HUD/LTM/Wax/Materials/M_UI_TextGlow_Add_Gradient_Orange_Nevada
     tiling=NoTile
     imageSize=<32 32>
     margin=<ltrb 0 0 0 0>
     outlineSettings {
      bUseBrushTransparency=false
      roundingType=HalfHeightRadius
      width=0
      color {colorUseRule=UseColor_Specified, specifiedColor=<rgba 0 0 0 0>}
      cornerRadii=<0 0 0 0>
     }
     tintColor {colorUseRule=UseColor_Specified, specifiedColor=<rgba 1 1 1 1>}
     uVRegion {bIsValid=false, max=<0 0>, min=<0 0>}
    }
    renderTransform {angle=0, scale=<1.325 1.75>, shear=<0 0>, translation=<0 0>}
   }
   slot {class=OverlaySlot, properties {horizontalAlignment=HAlign_Fill, verticalAlignment=VAlign_Fill}}
  }
  Image_Energy {
   widgetClassPath=&Image
   bIsVariable=true
   parent=Overlay_2
   widget=Image_Energy
   properties {
    brush {
     bIsDynamicallyLoaded=false
     drawAs=Image
     imageType=NoImage
     mirroring=NoMirror
     resourceName=None
     resourceObject=/Game/UI/Foundation/Textures/Icons/Stats/T-Icon-HealthBattery-64
     tiling=NoTile
     imageSize=<54 54>
     margin=<ltrb 0 0 0 0>
     outlineSettings {
      bUseBrushTransparency=false
      roundingType=HalfHeightRadius
      width=0
      color {colorUseRule=UseColor_Specified, specifiedColor=<rgba 0 0 0 0>}
      cornerRadii=<0 0 0 0>
     }
     tintColor {colorUseRule=UseColor_Specified, specifiedColor=<rgba 1 1 1 1>}
     uVRegion {bIsValid=false, max=<0 0>, min=<0 0>}
    }
    renderTransform {angle=0, scale=<1 1>, shear=<0 0>, translation=<0 -1>}
   }
   slot {class=OverlaySlot}
  }
  Overlay {widgetClassPath=&Overlay, parent=Switcher_Energy, widget=Overlay, slot {class=WidgetSwitcherSlot}}
  SizeBox_2 {
   widgetClassPath=&SizeBox
   parent=Overlay
   widget=SizeBox_2
   properties {bOverride_HeightOverride=true, bOverride_WidthOverride=true, heightOverride=54, widthOverride=54}
   slot {class=OverlaySlot, properties {horizontalAlignment=HAlign_Center, verticalAlignment=VAlign_Center}}
  }
  CRichText_Infinite {
   widgetClassPath=&CommonRichTextBlock
   parent=SizeBox_2
   widget=CRichText_Infinite
   properties {
    bOverrideDefaultStyle=true
    justification=Right
    lineHeightPercentage=0.85
    text="∞"
    textTransformPolicy=ToUpper
    defaultTextStyleOverride {
     overflowPolicy=Clip
     transformPolicy=None
     colorAndOpacity {colorUseRule=UseColor_Specified, specifiedColor=<rgba 1 1 1 1>}
     font {
      bForceMonospaced=false
      bMaterialIsStencil=false
      fontMaterial=null
      fontObject=&BurbankSmall
      letterSpacing=0
      monospacedWidth=1
      size=24
      skewAmount=0
      typefaceFontName=Bold
      outlineSettings {
       bApplyOutlineToDropShadows=false
       bMiteredCorners=false
       bSeparateFillAlpha=false
       outlineMaterial=null
       outlineSize=0
       outlineColor=<rgba 0 0 0 1>
      }
     }
     highlightColor {colorUseRule=UseColor_Specified, specifiedColor=<rgba 0 0 0 1>}
     highlightShape {
      bIsDynamicallyLoaded=false
      drawAs=Image
      imageType=NoImage
      mirroring=NoMirror
      resourceName=None
      resourceObject=null
      tiling=NoTile
      imageSize=<32 32>
      margin=<ltrb 0 0 0 0>
      outlineSettings {
       bUseBrushTransparency=false
       roundingType=HalfHeightRadius
       width=0
       color {colorUseRule=UseColor_Specified, specifiedColor=<rgba 0 0 0 0>}
       cornerRadii=<0 0 0 0>
      }
      tintColor {colorUseRule=UseColor_Specified, specifiedColor=<rgba 1 1 1 1>}
      uVRegion {bIsValid=false, max=<0 0>, min=<0 0>}
     }
     selectedBackgroundColor {colorUseRule=UseColor_Foreground, specifiedColor=<rgba 1 0 1 1>}
     shadowColorAndOpacity=<rgba 0 0 0 1>
     shadowOffset=<0 0>
     strikeBrush {
      bIsDynamicallyLoaded=false
      drawAs=Image
      imageType=NoImage
      mirroring=NoMirror
      resourceName=None
      resourceObject=null
      tiling=NoTile
      imageSize=<32 32>
      margin=<ltrb 0 0 0 0>
      outlineSettings {
       bUseBrushTransparency=false
       roundingType=HalfHeightRadius
       width=0
       color {colorUseRule=UseColor_Specified, specifiedColor=<rgba 0 0 0 0>}
       cornerRadii=<0 0 0 0>
      }
      tintColor {colorUseRule=UseColor_Specified, specifiedColor=<rgba 1 1 1 1>}
      uVRegion {bIsValid=false, max=<0 0>, min=<0 0>}
     }
     underlineBrush {
      bIsDynamicallyLoaded=false
      drawAs=Image
      imageType=NoImage
      mirroring=NoMirror
      resourceName=None
      resourceObject=null
      tiling=NoTile
      imageSize=<32 32>
      margin=<ltrb 0 0 0 0>
      outlineSettings {
       bUseBrushTransparency=false
       roundingType=HalfHeightRadius
       width=0
       color {colorUseRule=UseColor_Specified, specifiedColor=<rgba 0 0 0 0>}
       cornerRadii=<0 0 0 0>
      }
      tintColor {colorUseRule=UseColor_Specified, specifiedColor=<rgba 1 1 1 1>}
      uVRegion {bIsValid=false, max=<0 0>, min=<0 0>}
     }
    }
    renderTransform {angle=0, scale=<1 1>, shear=<0 0>, translation=<0 -5>}
   }
   slot {class=SizeBoxSlot, properties {horizontalAlignment=HAlign_Center, verticalAlignment=VAlign_Center}}
  }
  Overlay_3 {
   widgetClassPath=&Overlay
   parent=HorizontalBox_1
   widget=Overlay_3
   slot {class=HorizontalBoxSlot, properties {verticalAlignment=VAlign_Center}}
  }
  SizeBox_0 {
   widgetClassPath=&SizeBox
   parent=Overlay_3
   widget=SizeBox_0
   properties {bOverride_HeightOverride=true, bOverride_WidthOverride=true, heightOverride=24, widthOverride=300}
   slot {class=OverlaySlot, properties {horizontalAlignment=HAlign_Fill, verticalAlignment=VAlign_Center}}
  }
  Overlay_0 {widgetClassPath=&Overlay, parent=SizeBox_0, widget=Overlay_0, slot {class=SizeBoxSlot}}
  FillBar {
   widgetClassPath=&Image
   bIsVariable=true
   parent=Overlay_0
   widget=FillBar
   properties {brush {
     bIsDynamicallyLoaded=false
     drawAs=Image
     imageType=NoImage
     mirroring=NoMirror
     resourceName=None
     resourceObject=/BallerV2/HUD/Materials/M_UI_Baller_EnergyBar
     tiling=NoTile
     imageSize=<32 32>
     margin=<ltrb 0 0 0 0>
     outlineSettings {
      bUseBrushTransparency=false
      roundingType=HalfHeightRadius
      width=0
      color {colorUseRule=UseColor_Specified, specifiedColor=<rgba 0 0 0 0>}
      cornerRadii=<0 0 0 0>
     }
     tintColor {colorUseRule=UseColor_Specified, specifiedColor=<rgba 1 1 1 1>}
     uVRegion {bIsValid=false, max=<0 0>, min=<0 0>}
    }}
   slot {class=OverlaySlot, properties {horizontalAlignment=HAlign_Fill, verticalAlignment=VAlign_Fill}}
  }
  WarningMessaging {
   widgetClassPath=&Overlay
   bIsVariable=true
   parent=Overlay_3
   widget=WarningMessaging
   properties {renderOpacity=0, visibility=Collapsed, renderTransform {angle=0, scale=<1 1>, shear=<-16 0>, translation=<0 -56>}}
   slot {class=OverlaySlot, properties {verticalAlignment=VAlign_Center}}
  }
  Overlay_4 {
   widgetClassPath=&Overlay
   parent=WarningMessaging
   widget=Overlay_4
   slot {class=OverlaySlot, properties {horizontalAlignment=HAlign_Fill, verticalAlignment=VAlign_Fill, padding=<ltrb -8 -8 -8 -8>}}
  }
  TextContainerLow {
   widgetClassPath=&Overlay
   parent=Overlay_4
   widget=TextContainerLow
   properties {renderTransform {angle=0, scale=<1 1>, shear=<0 0>, translation=<8 -8>}}
   slot {class=OverlaySlot}
  }
  AdditonalOuterRing {
   widgetClassPath=&Image
   bIsVariable=true
   parent=TextContainerLow
   widget=AdditonalOuterRing
   properties {
    renderOpacity=0.25
    brush {
     bIsDynamicallyLoaded=false
     drawAs=Box
     imageType=NoImage
     mirroring=NoMirror
     resourceName=None
     resourceObject=/Game/Athena/HUD/Vehicles/Materials/M_UI_OstrichWarningText_Outline
     tiling=NoTile
     imageSize=<16 16>
     margin=<ltrb 0.5 0.5 0.5 0.5>
     outlineSettings {
      bUseBrushTransparency=false
      roundingType=HalfHeightRadius
      width=0
      color {colorUseRule=UseColor_Specified, specifiedColor=<rgba 0 0 0 0>}
      cornerRadii=<0 0 0 0>
     }
     tintColor {colorUseRule=UseColor_Specified, specifiedColor=<rgba 1 1 1 1>}
     uVRegion {bIsValid=false, max=<0 0>, min=<0 0>}
    }
   }
   slot {class=OverlaySlot, properties {horizontalAlignment=HAlign_Fill, verticalAlignment=VAlign_Fill, padding=<ltrb -8 -8 -8 -8>}}
  }
  Image_3 {
   widgetClassPath=&Image
   bIsVariable=true
   parent=TextContainerLow
   widget=Image_3
   properties {
    brush {
     bIsDynamicallyLoaded=false
     drawAs=Box
     imageType=NoImage
     mirroring=NoMirror
     resourceName=None
     resourceObject=/Game/Athena/HUD/Vehicles/Materials/M_UI_OstrichWarningText_Backing
     tiling=NoTile
     imageSize=<8 8>
     margin=<ltrb 0.5 0.5 0.5 0.5>
     outlineSettings {
      bUseBrushTransparency=false
      roundingType=HalfHeightRadius
      width=0
      color {colorUseRule=UseColor_Specified, specifiedColor=<rgba 0 0 0 0>}
      cornerRadii=<0 0 0 0>
     }
     tintColor {colorUseRule=UseColor_Specified, specifiedColor=<rgba 1 1 1 1>}
     uVRegion {bIsValid=false, max=<0 0>, min=<0 0>}
    }
    renderTransformPivot=<1 0.5>
   }
   slot {class=OverlaySlot, properties {horizontalAlignment=HAlign_Fill, verticalAlignment=VAlign_Fill}}
  }
  Image_4 {
   widgetClassPath=&Image
   bIsVariable=true
   parent=TextContainerLow
   widget=Image_4
   properties {
    renderOpacity=0.5
    brush {
     bIsDynamicallyLoaded=false
     drawAs=Box
     imageType=NoImage
     mirroring=NoMirror
     resourceName=None
     resourceObject=&M_UI_OstrichWarningText_Outline
     tiling=NoTile
     imageSize=<8 8>
     margin=<ltrb 0.5 0.5 0.5 0.5>
     outlineSettings {
      bUseBrushTransparency=false
      roundingType=HalfHeightRadius
      width=0
      color {colorUseRule=UseColor_Specified, specifiedColor=<rgba 0 0 0 0>}
      cornerRadii=<0 0 0 0>
     }
     tintColor {colorUseRule=UseColor_Specified, specifiedColor=<rgba 1 1 1 1>}
     uVRegion {bIsValid=false, max=<0 0>, min=<0 0>}
    }
   }
   slot {class=OverlaySlot, properties {horizontalAlignment=HAlign_Fill, verticalAlignment=VAlign_Fill}}
  }
  Text_NoEnergy {
   widgetClassPath=&CommonTextBlock
   bIsVariable=true
   parent=TextContainerLow
   widget=Text_NoEnergy
   properties {
    justification=Center
    text="Low Energy!"
    textTransformPolicy=ToUpper
    colorAndOpacity {colorUseRule=UseColor_Specified, specifiedColor=<rgba 0 0 0 0.5>}
    font {
     bForceMonospaced=false
     bMaterialIsStencil=false
     fontMaterial=null
     fontObject=&BurbankSmall
     letterSpacing=100
     monospacedWidth=1
     size=13
     skewAmount=0
     typefaceFontName=Black
     outlineSettings {
      bApplyOutlineToDropShadows=false
      bMiteredCorners=false
      bSeparateFillAlpha=true
      outlineMaterial=null
      outlineSize=2
      outlineColor=<rgba 1 1 1 0.125>
     }
    }
    shadowOffset=<0 0>
   }
   slot {class=OverlaySlot, properties {horizontalAlignment=HAlign_Fill, verticalAlignment=VAlign_Center, padding=<ltrb 12 2 12 2>}}
  }
  LargeGlow {
   widgetClassPath=&Image
   bIsVariable=true
   parent=TextContainerLow
   widget=LargeGlow
   properties {
    renderOpacity=0.25
    visibility=SelfHitTestInvisible
    brush {
     bIsDynamicallyLoaded=false
     drawAs=Box
     imageType=NoImage
     mirroring=NoMirror
     resourceName=None
     resourceObject=&M_UI_TextGlow_Add_Gradient_Orange_Nevada
     tiling=NoTile
     imageSize=<64 64>
     margin=<ltrb 0.5 0.5 0.5 0.5>
     outlineSettings {
      bUseBrushTransparency=false
      roundingType=HalfHeightRadius
      width=0
      color {colorUseRule=UseColor_Specified, specifiedColor=<rgba 0 0 0 0>}
      cornerRadii=<0 0 0 0>
     }
     tintColor {colorUseRule=UseColor_Specified, specifiedColor=<rgba 1 1 1 1>}
     uVRegion {bIsValid=false, max=<0 0>, min=<0 0>}
    }
   }
   slot {class=OverlaySlot, properties {horizontalAlignment=HAlign_Fill, verticalAlignment=VAlign_Fill, padding=<ltrb -24 -24 -24 -24>}}
  }
  TextContainerEject {
   widgetClassPath=&Overlay
   parent=Overlay_4
   widget=TextContainerEject
   properties {renderOpacity=0, renderTransform {angle=0, scale=<1 1>, shear=<0 0>, translation=<8 -8>}}
   slot {class=OverlaySlot}
  }
  AdditonalOuterRing_1 {
   widgetClassPath=&Image
   bIsVariable=true
   parent=TextContainerEject
   widget=AdditonalOuterRing_1
   properties {
    renderOpacity=0.25
    brush {
     bIsDynamicallyLoaded=false
     drawAs=Box
     imageType=NoImage
     mirroring=NoMirror
     resourceName=None
     resourceObject=&M_UI_OstrichWarningText_Outline
     tiling=NoTile
     imageSize=<16 16>
     margin=<ltrb 0.5 0.5 0.5 0.5>
     outlineSettings {
      bUseBrushTransparency=false
      roundingType=HalfHeightRadius
      width=0
      color {colorUseRule=UseColor_Specified, specifiedColor=<rgba 0 0 0 0>}
      cornerRadii=<0 0 0 0>
     }
     tintColor {colorUseRule=UseColor_Specified, specifiedColor=<rgba 1 1 1 1>}
     uVRegion {bIsValid=false, max=<0 0>, min=<0 0>}
    }
   }
   slot {class=OverlaySlot, properties {horizontalAlignment=HAlign_Fill, verticalAlignment=VAlign_Fill, padding=<ltrb -8 -8 -8 -8>}}
  }
  Image {
   widgetClassPath=&Image
   bIsVariable=true
   parent=TextContainerEject
   widget=Image
   properties {
    brush {
     bIsDynamicallyLoaded=false
     drawAs=Box
     imageType=NoImage
     mirroring=NoMirror
     resourceName=None
     resourceObject=/Game/Athena/HUD/Vehicles/Materials/M_UI_OstrichWarningText_Backing_Red
     tiling=NoTile
     imageSize=<8 8>
     margin=<ltrb 0.5 0.5 0.5 0.5>
     outlineSettings {
      bUseBrushTransparency=false
      roundingType=HalfHeightRadius
      width=0
      color {colorUseRule=UseColor_Specified, specifiedColor=<rgba 0 0 0 0>}
      cornerRadii=<0 0 0 0>
     }
     tintColor {colorUseRule=UseColor_Specified, specifiedColor=<rgba 1 1 1 1>}
     uVRegion {bIsValid=false, max=<0 0>, min=<0 0>}
    }
    renderTransformPivot=<1 0.5>
   }
   slot {class=OverlaySlot, properties {horizontalAlignment=HAlign_Fill, verticalAlignment=VAlign_Fill}}
  }
  Image_1 {
   widgetClassPath=&Image
   bIsVariable=true
   parent=TextContainerEject
   widget=Image_1
   properties {
    renderOpacity=0.5
    brush {
     bIsDynamicallyLoaded=false
     drawAs=Box
     imageType=NoImage
     mirroring=NoMirror
     resourceName=None
     resourceObject=&M_UI_OstrichWarningText_Outline
     tiling=NoTile
     imageSize=<8 8>
     margin=<ltrb 0.5 0.5 0.5 0.5>
     outlineSettings {
      bUseBrushTransparency=false
      roundingType=HalfHeightRadius
      width=0
      color {colorUseRule=UseColor_Specified, specifiedColor=<rgba 0 0 0 0>}
      cornerRadii=<0 0 0 0>
     }
     tintColor {colorUseRule=UseColor_Specified, specifiedColor=<rgba 1 1 1 1>}
     uVRegion {bIsValid=false, max=<0 0>, min=<0 0>}
    }
   }
   slot {class=OverlaySlot, properties {horizontalAlignment=HAlign_Fill, verticalAlignment=VAlign_Fill}}
  }
  Text_Eject {
   widgetClassPath=&CommonTextBlock
   bIsVariable=true
   parent=TextContainerEject
   widget=Text_Eject
   properties {
    justification=Center
    text=ENERGY DEPLETED
    textTransformPolicy=ToUpper
    colorAndOpacity {colorUseRule=UseColor_Specified, specifiedColor=<rgba 0 0 0 0.5>}
    font {
     bForceMonospaced=false
     bMaterialIsStencil=false
     fontMaterial=null
     fontObject=&BurbankSmall
     letterSpacing=100
     monospacedWidth=1
     size=13
     skewAmount=0
     typefaceFontName=Black
     outlineSettings {
      bApplyOutlineToDropShadows=false
      bMiteredCorners=false
      bSeparateFillAlpha=true
      outlineMaterial=null
      outlineSize=2
      outlineColor=<rgba 1 1 1 0.125>
     }
    }
    shadowOffset=<0 0>
   }
   slot {class=OverlaySlot, properties {horizontalAlignment=HAlign_Fill, verticalAlignment=VAlign_Center, padding=<ltrb 12 2 12 2>}}
  }
  LargeGlow_1 {
   widgetClassPath=&Image
   bIsVariable=true
   parent=TextContainerEject
   widget=LargeGlow_1
   properties {
    renderOpacity=0.25
    visibility=SelfHitTestInvisible
    brush {
     bIsDynamicallyLoaded=false
     drawAs=Box
     imageType=NoImage
     mirroring=NoMirror
     resourceName=None
     resourceObject=/Game/Athena/HUD/LTM/Wax/Materials/M_UI_BallerSelftDestruct_Textglow
     tiling=NoTile
     imageSize=<64 64>
     margin=<ltrb 0.5 0.5 0.5 0.5>
     outlineSettings {
      bUseBrushTransparency=false
      roundingType=HalfHeightRadius
      width=0
      color {colorUseRule=UseColor_Specified, specifiedColor=<rgba 0 0 0 0>}
      cornerRadii=<0 0 0 0>
     }
     tintColor {colorUseRule=UseColor_Specified, specifiedColor=<rgba 1 1 1 1>}
     uVRegion {bIsValid=false, max=<0 0>, min=<0 0>}
    }
   }
   slot {class=OverlaySlot, properties {horizontalAlignment=HAlign_Fill, verticalAlignment=VAlign_Fill, padding=<ltrb -24 -24 -24 -24>}}
  }
  AlertIcon {
   widgetClassPath=&Overlay
   parent=TextContainerEject
   widget=AlertIcon
   properties {visibility=Collapsed, renderTransform {angle=0, scale=<1 1>, shear=<8 0>, translation=<-32 -40>}}
   slot {class=OverlaySlot}
  }
  Image_210 {
   widgetClassPath=&Image
   bIsVariable=true
   parent=AlertIcon
   widget=Image_210
   properties {brush {
     bIsDynamicallyLoaded=false
     drawAs=Image
     imageType=NoImage
     mirroring=NoMirror
     resourceName=None
     resourceObject=&M_UI_OstrichWarningText_Backing
     tiling=NoTile
     imageSize=<32 32>
     margin=<ltrb 0 0 0 0>
     outlineSettings {
      bUseBrushTransparency=false
      roundingType=HalfHeightRadius
      width=0
      color {colorUseRule=UseColor_Specified, specifiedColor=<rgba 0 0 0 0>}
      cornerRadii=<0 0 0 0>
     }
     tintColor {colorUseRule=UseColor_Specified, specifiedColor=<rgba 1 1 1 1>}
     uVRegion {bIsValid=false, max=<0 0>, min=<0 0>}
    }}
   slot {class=OverlaySlot}
  }
  CommonTextBlock_245 {
   widgetClassPath=&CommonTextBlock
   parent=AlertIcon
   widget=CommonTextBlock_245
   properties {
    text="!"
    font {
     bForceMonospaced=false
     bMaterialIsStencil=false
     fontMaterial=null
     fontObject=&BurbankSmall
     letterSpacing=0
     monospacedWidth=1
     size=17
     skewAmount=0
     typefaceFontName=Black
     outlineSettings {
      bApplyOutlineToDropShadows=false
      bMiteredCorners=false
      bSeparateFillAlpha=true
      outlineMaterial=null
      outlineSize=2
      outlineColor=<rgba 1 1 1 0.125>
     }
    }
    shadowOffset=<0 0>
   }
   slot {class=OverlaySlot, properties {horizontalAlignment=HAlign_Center, verticalAlignment=VAlign_Center}}
  }
  LargeGlow_2 {
   widgetClassPath=&Image
   bIsVariable=true
   parent=AlertIcon
   widget=LargeGlow_2
   properties {
    renderOpacity=0.25
    visibility=SelfHitTestInvisible
    brush {
     bIsDynamicallyLoaded=false
     drawAs=Image
     imageType=NoImage
     mirroring=NoMirror
     resourceName=None
     resourceObject=&M_UI_BallerSelftDestruct_Textglow
     tiling=NoTile
     imageSize=<64 64>
     margin=<ltrb 0.5 0.5 0.5 0.5>
     outlineSettings {
      bUseBrushTransparency=false
      roundingType=HalfHeightRadius
      width=0
      color {colorUseRule=UseColor_Specified, specifiedColor=<rgba 0 0 0 0>}
      cornerRadii=<0 0 0 0>
     }
     tintColor {colorUseRule=UseColor_Specified, specifiedColor=<rgba 1 1 1 1>}
     uVRegion {bIsValid=false, max=<0 0>, min=<0 0>}
    }
   }
   slot {class=OverlaySlot, properties {horizontalAlignment=HAlign_Fill, verticalAlignment=VAlign_Fill, padding=<ltrb -24 -24 -24 -24>}}
  }
  TriangleArrow {
   widgetClassPath=&Image
   bIsVariable=true
   parent=WarningMessaging
   widget=TriangleArrow
   properties {
    brush {
     bIsDynamicallyLoaded=false
     drawAs=Image
     imageType=NoImage
     mirroring=NoMirror
     resourceName=None
     resourceObject=/Game/UI/SocialPanel/Art/Textures/T_UI_Icon_TreeArrow_64
     tiling=NoTile
     imageSize=<16 16>
     margin=<ltrb 0 0 0 0>
     outlineSettings {
      bUseBrushTransparency=false
      roundingType=HalfHeightRadius
      width=0
      color {colorUseRule=UseColor_Specified, specifiedColor=<rgba 0 0 0 0>}
      cornerRadii=<0 0 0 0>
     }
     tintColor {colorUseRule=UseColor_Specified, specifiedColor=<rgba 1 1 1 1>}
     uVRegion {bIsValid=false, max=<0 0>, min=<0 0>}
    }
    colorAndOpacity=<rgba 1 0.4167 0 1>
    renderTransform {angle=108, scale=<1 1>, shear=<0 0>, translation=<-8 24>}
   }
   slot {class=OverlaySlot, properties {verticalAlignment=VAlign_Bottom}}
  }
  BoostGlowBorder {
   widgetClassPath=&Border
   bIsVariable=true
   parent=Overlay_3
   widget=BoostGlowBorder
   properties {
    horizontalAlignment=HAlign_Left
    visibility=Collapsed
    background {
     bIsDynamicallyLoaded=false
     drawAs=NoDrawType
     imageType=NoImage
     mirroring=NoMirror
     resourceName=None
     resourceObject=null
     tiling=NoTile
     imageSize=<32 32>
     margin=<ltrb 0 0 0 0>
     outlineSettings {
      bUseBrushTransparency=false
      roundingType=HalfHeightRadius
      width=0
      color {colorUseRule=UseColor_Specified, specifiedColor=<rgba 0 0 0 0>}
      cornerRadii=<0 0 0 0>
     }
     tintColor {colorUseRule=UseColor_Specified, specifiedColor=<rgba 1 1 1 1>}
     uVRegion {bIsValid=false, max=<0 0>, min=<0 0>}
    }
    padding=<ltrb 0 0 0 0>
    renderTransformPivot=<0 0.5>
   }
   slot {class=OverlaySlot, properties {verticalAlignment=VAlign_Fill}}
  }
  SizeBox_1 {
   widgetClassPath=&SizeBox
   parent=BoostGlowBorder
   widget=SizeBox_1
   properties {bOverride_WidthOverride=true, widthOverride=300}
   slot {class=BorderSlot, properties {horizontalAlignment=HAlign_Left, padding=<ltrb 0 0 0 0>}}
  }
  BoostGlow {
   widgetClassPath=&Image
   bIsVariable=true
   parent=SizeBox_1
   widget=BoostGlow
   properties {
    brush {
     bIsDynamicallyLoaded=false
     drawAs=Box
     imageType=NoImage
     mirroring=NoMirror
     resourceName=None
     resourceObject=/Game/Athena/HUD/LTM/Wax/Materials/M_UI_BallerBoostGlow
     tiling=NoTile
     imageSize=<72 72>
     margin=<ltrb 0.5 0.5 0.5 0.5>
     outlineSettings {
      bUseBrushTransparency=false
      roundingType=HalfHeightRadius
      width=0
      color {colorUseRule=UseColor_Specified, specifiedColor=<rgba 0 0 0 0>}
      cornerRadii=<0 0 0 0>
     }
     tintColor {colorUseRule=UseColor_Specified, specifiedColor=<rgba 1 1 1 1>}
     uVRegion {bIsValid=false, max=<0 0>, min=<0 0>}
    }
    renderTransformPivot=<0 0.5>
   }
   slot {class=SizeBoxSlot, properties {padding=<ltrb -32 -32 -32 -32>}}
  }
  Overlay_BoostSparks {
   widgetClassPath=&Overlay
   bInherited=true
   bIsVariable=true
   parent=Overlay_3
   widget=Overlay_BoostSparks
   properties {visibility=Collapsed, renderTransform {angle=90, scale=<1 1>, shear=<0 0>, translation=<0 0>}, renderTransformPivot=<0 0.5>}
   slot {class=OverlaySlot, properties {horizontalAlignment=HAlign_Right, verticalAlignment=VAlign_Center}}
  }
  Overlay_47 {
   widgetClassPath=&Overlay
   parent=Overlay_BoostSparks
   widget=Overlay_47
   properties {visibility=HitTestInvisible, renderTransform {angle=172, scale=<1 1>, shear=<0 0>, translation=<2 8>}}
   slot {class=OverlaySlot}
  }
  Image_Sparks {
   widgetClassPath=&Image
   bIsVariable=true
   parent=Overlay_47
   widget=Image_Sparks
   properties {
    visibility=HitTestInvisible
    brush {
     bIsDynamicallyLoaded=false
     drawAs=Image
     imageType=NoImage
     mirroring=NoMirror
     resourceName=None
     resourceObject=/Game/Athena/UI/InGame/Interrogation/Materials/M_UI_Starburst_Sparks
     tiling=NoTile
     imageSize=<295.3884 400>
     margin=<ltrb 0 0 0 0>
     outlineSettings {
      bUseBrushTransparency=false
      roundingType=HalfHeightRadius
      width=0
      color {colorUseRule=UseColor_Specified, specifiedColor=<rgba 0 0 0 0>}
      cornerRadii=<0 0 0 0>
     }
     tintColor {colorUseRule=UseColor_Specified, specifiedColor=<rgba 1 0.3682 0.0743 1>}
     uVRegion {bIsValid=false, max=<0 0>, min=<0 0>}
    }
   }
   slot {class=OverlaySlot, properties {horizontalAlignment=HAlign_Center, verticalAlignment=VAlign_Center, padding=<ltrb -200 -200 -200 -200>}}
  }
  Image_Sparks_1 {
   widgetClassPath=&Image
   bIsVariable=true
   parent=Overlay_47
   widget=Image_Sparks_1
   properties {
    visibility=HitTestInvisible
    brush {
     bIsDynamicallyLoaded=false
     drawAs=Image
     imageType=NoImage
     mirroring=NoMirror
     resourceName=None
     resourceObject=/Game/Athena/UI/InGame/Interrogation/Materials/M_UI_Starburst_Sparks_Inst
     tiling=NoTile
     imageSize=<120 400>
     margin=<ltrb 0 0 0 0>
     outlineSettings {
      bUseBrushTransparency=false
      roundingType=HalfHeightRadius
      width=0
      color {colorUseRule=UseColor_Specified, specifiedColor=<rgba 0 0 0 0>}
      cornerRadii=<0 0 0 0>
     }
     tintColor {colorUseRule=UseColor_Specified, specifiedColor=<rgba 1 0.7183 0.651 1>}
     uVRegion {bIsValid=false, max=<0 0>, min=<0 0>}
    }
   }
   slot {class=OverlaySlot, properties {horizontalAlignment=HAlign_Center, verticalAlignment=VAlign_Center, padding=<ltrb -200 -200 -200 -200>}}
  }
  Image_LargeCone {
   widgetClassPath=&Image
   bIsVariable=true
   parent=Overlay_47
   widget=Image_LargeCone
   properties {
    visibility=HitTestInvisible
    brush {
     bIsDynamicallyLoaded=false
     drawAs=Image
     imageType=NoImage
     mirroring=NoMirror
     resourceName=None
     resourceObject=/Game/Athena/HUD/Vehicles/Materials/M_UI_SparksGlow_Inst
     tiling=NoTile
     imageSize=<150 219.5743>
     margin=<ltrb 0 0 0 0>
     outlineSettings {
      bUseBrushTransparency=false
      roundingType=HalfHeightRadius
      width=0
      color {colorUseRule=UseColor_Specified, specifiedColor=<rgba 0 0 0 0>}
      cornerRadii=<0 0 0 0>
     }
     tintColor {colorUseRule=UseColor_Specified, specifiedColor=<rgba 1 1 1 1>}
     uVRegion {bIsValid=false, max=<0 0>, min=<0 0>}
    }
    renderTransform {angle=0, scale=<1 1>, shear=<0 0>, translation=<0 44>}
   }
   slot {class=OverlaySlot, properties {horizontalAlignment=HAlign_Center, verticalAlignment=VAlign_Center, padding=<ltrb -200 -200 -200 -200>}}
  }
  Widget_FuelGauge {widgetClassPath=/Script/FortniteUI.FortVehicleFuelGauge, bInherited=true, widget=null}
  HorizontalBox_FuelCluster {widgetClassPath=&HorizontalBox, bInherited=true, widget=null}
  Image_Speedometer {widgetClassPath=&Image, bInherited=true, widget=null}
  Image_BoostGlow {widgetClassPath=&Image, bInherited=true, widget=null}
  Image_BoostMeter {widgetClassPath=&Image, bInherited=true, widget=null}
  NumericText_Speed {widgetClassPath=/Script/CommonUI.CommonNumericTextBlock, bInherited=true, widget=null}
  Overlay_Gauge {widgetClassPath=&Overlay, bInherited=true, widget=null}
  Text_ItemName {widgetClassPath=&CommonTextBlock, bInherited=true, widget=null}
  TutorialHighlight {widgetClassPath=&TutorialHighlight}
 ]
}
