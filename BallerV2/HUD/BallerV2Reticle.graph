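# Baller vehicle dashboard reticle widget for the towhook grapple, showing lock-on, attached, cooldown and disabled-vehicle states.
# BindTo hooks the vehicle's out-of-fuel and attach/detach delegates plus the seat weapon's target updates, which feed Update and SetLockedOnState.
# Those project the world target to screen, play lock-on and swing animations forward or reversed, and drive the CooldownRing material.
# Out of fuel tints the reticle; a Vehicle.Disabled tag toggles a warning.
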WidgetBlueprint /BallerV2/HUD/BallerV2Reticle
animations [
 Swinging {
  binding_count=4
  length_seconds=0.2
  object=&self:Swinging
  sequence {bindings [
    Reticle_Surrounding_BaseDarken {
     id=#1
     possessed_class=Image
     tracks [{
       type=MovieSceneWidgetMaterialTrack
       sections [{
         type=MovieSceneParameterSection
         channels [
          AnimateScale {type=MovieSceneScriptingFloatChannel, keys [[0,0],[4,1]]}
          CrosshairGap {type=MovieSceneScriptingFloatChannel, const [1,1]}
         ]
        }]
      }]
    }
    Reticle_Surrounding_Shadow {
     id=#2
     possessed_class=Image
     tracks [{
       type=MovieSceneWidgetMaterialTrack
       sections [{
         type=MovieSceneParameterSection
         channels [
          AnimateScale {type=MovieSceneScriptingFloatChannel, keys [[0,0],[4,1]]}
          CrosshairGap {type=MovieSceneScriptingFloatChannel, const [1,1]}
         ]
        }]
      }]
    }
    Reticle_Spinners_AttachedFast {
     id=#3
     possessed_class=Image
     tracks [
      {
       type=MovieScene2DTransformTrack
       sections [{
         type=MovieScene2DTransformSection
         channels [
          Translation.X {type=MovieSceneScriptingFloatChannel, const [0,1]}
          Translation.Y {type=MovieSceneScriptingFloatChannel, const [0,1]}
          Angle {type=MovieSceneScriptingFloatChannel, const [0,1]}
          Scale.X {type=MovieSceneScriptingFloatChannel, keys [[0,1],[4,0.75]]}
          Scale.Y {type=MovieSceneScriptingFloatChannel, keys [[0,1],[4,0.75]]}
          Shear.X {type=MovieSceneScriptingFloatChannel, const [0,1]}
          Shear.Y {type=MovieSceneScriptingFloatChannel, const [0,1]}
         ]
        }]
      }
      {
       type=MovieSceneFloatTrack
       display_name=Render Opacity
       sections [{type=MovieSceneFloatSection, channels [None {type=MovieSceneScriptingFloatChannel, keys [[0,0],[2,1]]}]}]
      }
     ]
    }
    Reticle_Surrounding {
     id=#4
     possessed_class=Image
     tracks [{
       type=MovieSceneWidgetMaterialTrack
       sections [{
         type=MovieSceneParameterSection
         channels [
          AnimateScale {type=MovieSceneScriptingFloatChannel, keys [[0,0],[4,1]]}
          AnimateAlpha {type=MovieSceneScriptingFloatChannel, keys [[0,0],[4,1]]}
          CrosshairGap {type=MovieSceneScriptingFloatChannel, const [1,1]}
         ]
        }]
      }]
    }
   ]}
 }
 CantUse {
  binding_count=5
  length_seconds=0.2
  object=&self:CantUse
  sequence {bindings [
    AnimBoundSizeControl {
     id=#5
     possessed_class=Border
     tracks [{
       type=MovieScene2DTransformTrack
       sections [{
         type=MovieScene2DTransformSection
         channels [
          Translation.X {type=MovieSceneScriptingFloatChannel, const [0,1]}
          Translation.Y {type=MovieSceneScriptingFloatChannel, const [0,1]}
          Angle {type=MovieSceneScriptingFloatChannel, keys [[0,0],[4,45]]}
          Scale.X {type=MovieSceneScriptingFloatChannel, keys [[0,1],[4,0.75]]}
          Scale.Y {type=MovieSceneScriptingFloatChannel, keys [[0,1],[4,0.75]]}
          Shear.X {type=MovieSceneScriptingFloatChannel, const [0,1]}
          Shear.Y {type=MovieSceneScriptingFloatChannel, const [0,1]}
         ]
        }]
      }]
    }
    Reticle_Surrounding_BaseDarken {
     id=#6
     possessed_class=Image
     tracks [{
       type=MovieSceneWidgetMaterialTrack
       sections [{
         type=MovieSceneParameterSection
         channels [
          CrosshairGap {type=MovieSceneScriptingFloatChannel, keys [[0,0.885],[4,1]]}
          AnimateScale {type=MovieSceneScriptingFloatChannel, const [0,1]}
         ]
        }]
      }]
    }
    Reticle_Surrounding {
     id=#7
     possessed_class=Image
     tracks [{
       type=MovieSceneWidgetMaterialTrack
       sections [{
         type=MovieSceneParameterSection
         channels [
          CrosshairGap {type=MovieSceneScriptingFloatChannel, keys [[0,0.885],[4,1]]}
          AnimateDisabledAlpha {type=MovieSceneScriptingFloatChannel, const [1,1]}
          AnimateScale {type=MovieSceneScriptingFloatChannel, const [0,1]}
         ]
        }]
      }]
    }
    Reticle_Surrounding_Disabled {
     id=#8
     possessed_class=Image
     tracks [{
       type=MovieSceneWidgetMaterialTrack
       sections [{
         type=MovieSceneParameterSection
         channels [
          CrosshairGap {type=MovieSceneScriptingFloatChannel, keys [[0,0.885],[4,1]]}
          AnimateScale {type=MovieSceneScriptingFloatChannel, const [0,1]}
         ]
        }]
      }]
    }
    Reticle_Surrounding_Shadow {
     id=#9
     possessed_class=Image
     tracks [{
       type=MovieSceneWidgetMaterialTrack
       sections [{
         type=MovieSceneParameterSection
         channels [
          CrosshairGap {type=MovieSceneScriptingFloatChannel, keys [[0,0.885],[4,1]]}
          AnimateScale {type=MovieSceneScriptingFloatChannel, const [0,1]}
         ]
        }]
      }]
    }
   ]}
 }
 LockedOn {
  binding_count=4
  length_seconds=0.1
  object=&self:LockedOn
  sequence {bindings [
    Reticle_Surrounding {
     id=#10
     possessed_class=Image
     tracks [{
       type=MovieSceneWidgetMaterialTrack
       sections [{
         type=MovieSceneParameterSection
         channels [
          CrosshairGap {type=MovieSceneScriptingFloatChannel, keys [[0,0.885],[2,1]]}
          AnimateDisabledAlpha {type=MovieSceneScriptingFloatChannel, keys [[0,1],[2,0]]}
          LockedOn {type=MovieSceneScriptingFloatChannel, keys [[0,0],[2,1]]}
         ]
        }]
      }]
    }
    Reticle_Surrounding_Shadow {
     id=#11
     possessed_class=Image
     tracks [{
       type=MovieSceneWidgetMaterialTrack
       sections [{type=MovieSceneParameterSection, channels [CrosshairGap {type=MovieSceneScriptingFloatChannel, keys [[0,0.885],[2,1]]}]}]
      }]
    }
    Bracket {
     id=#12
     possessed_class=Overlay
     tracks [
      {
       type=MovieScene2DTransformTrack
       sections [{
         type=MovieScene2DTransformSection
         channels [
          Scale.X {type=MovieSceneScriptingFloatChannel, keys [[0,1.25],[2,1]]}
          Scale.Y {type=MovieSceneScriptingFloatChannel, keys [[0,1.25],[2,1]]}
         ]
        }]
      }
      {
       type=MovieSceneFloatTrack
       display_name=Render Opacity
       sections [{type=MovieSceneFloatSection, channels [None {type=MovieSceneScriptingFloatChannel, keys [[0,0.5],[2,1]]}]}]
      }
     ]
    }
    Reticle_Surrounding_BaseDarken {
     id=#13
     possessed_class=Image
     tracks [{
       type=MovieSceneWidgetMaterialTrack
       sections [{type=MovieSceneParameterSection, channels [CrosshairGap {type=MovieSceneScriptingFloatChannel, keys [[0,0.885],[2,1]]}]}]
      }]
    }
   ]}
 }
 Cooldown_Intro {
  binding_count=1
  length_seconds=0.4
  object=&self:Cooldown_Intro
  sequence {bindings [CooldownRing {
     id=#14
     possessed_class=Image
     tracks [
      {
       type=MovieScene2DTransformTrack
       sections [{
         type=MovieScene2DTransformSection
         channels [
          Scale.X {type=MovieSceneScriptingFloatChannel, keys [[0,0.5],[8,1]]}
          Scale.Y {type=MovieSceneScriptingFloatChannel, keys [[0,0.5],[8,1]]}
         ]
        }]
      }
      {
       type=MovieSceneFloatTrack
       display_name=Render Opacity
       sections [{type=MovieSceneFloatSection, channels [None {type=MovieSceneScriptingFloatChannel, keys [[0,0],[5,1]]}]}]
      }
     ]
    }]}
 }
 LockedOnNew {
  binding_count=6
  length_seconds=0.1
  object=&self:LockedOnNew
  sequence {bindings [
    Reticle_Surrounding {
     id=#15
     possessed_class=Image
     tracks [{
       type=MovieSceneWidgetMaterialTrack
       sections [{
         type=MovieSceneParameterSection
         channels [
          LockedOn {type=MovieSceneScriptingFloatChannel, keys [[0,0],[2,1]]}
          AnimateAlpha {type=MovieSceneScriptingFloatChannel, keys [[0,0],[2,1]]}
         ]
        }]
      }]
    }
    Reticle_Surrounding_Shadow {
     id=#16
     possessed_class=Image
     tracks [{
       type=MovieSceneWidgetMaterialTrack
       sections [{type=MovieSceneParameterSection, channels [LockedOn {type=MovieSceneScriptingFloatChannel, keys [[0,0],[2,1]]}]}]
      }]
    }
    Reticle_Surrounding_BaseDarken {
     id=#17
     possessed_class=Image
     tracks [{
       type=MovieSceneWidgetMaterialTrack
       sections [{type=MovieSceneParameterSection, channels [LockedOn {type=MovieSceneScriptingFloatChannel, keys [[0,0],[2,1]]}]}]
      }]
    }
    Reticle_Spinners_AttachedFast {
     id=#18
     possessed_class=Image
     tracks [{
       type=MovieSceneFloatTrack
       display_name=Render Opacity
       sections [{type=MovieSceneFloatSection, channels [None {type=MovieSceneScriptingFloatChannel, const [0,1]}]}]
      }]
    }
    AnimBoundSizeControl {
     id=#19
     possessed_class=Border
     tracks [{
       type=MovieScene2DTransformTrack
       sections [{
         type=MovieScene2DTransformSection
         channels [
          Scale.X {type=MovieSceneScriptingFloatChannel, keys [[0,1.1875],[2,1]]}
          Scale.Y {type=MovieSceneScriptingFloatChannel, keys [[0,1.1875],[2,1]]}
         ]
        }]
      }]
    }
    Bracket {
     id=#20
     possessed_class=Overlay
     tracks [{
       type=MovieScene2DTransformTrack
       sections [{
         type=MovieScene2DTransformSection
         channels [
          Angle {type=MovieSceneScriptingFloatChannel, const [0,1]}
          Scale.X {type=MovieSceneScriptingFloatChannel, const [1,1]}
          Scale.Y {type=MovieSceneScriptingFloatChannel, const [1,1]}
         ]
        }]
      }]
    }
   ]}
 }
 SwingingNew {
  binding_count=6
  length_seconds=0.2
  object=&self:SwingingNew
  sequence {bindings [
    Reticle_Surrounding {
     id=#15
     possessed_class=Image
     tracks [{
       type=MovieSceneWidgetMaterialTrack
       sections [{
         type=MovieSceneParameterSection
         channels [
          LockedOn {type=MovieSceneScriptingFloatChannel, const [1,1]}
          AnimateAlpha {type=MovieSceneScriptingFloatChannel, const [1,1]}
         ]
        }]
      }]
    }
    Reticle_Surrounding_BaseDarken {
     id=#21
     possessed_class=Image
     tracks [{
       type=MovieSceneWidgetMaterialTrack
       sections [{type=MovieSceneParameterSection, channels [LockedOn {type=MovieSceneScriptingFloatChannel, const [1,1]}]}]
      }]
    }
    Bracket {
     id=#22
     possessed_class=Overlay
     tracks [{
       type=MovieScene2DTransformTrack
       sections [{
         type=MovieScene2DTransformSection
         channels [
          Angle {type=MovieSceneScriptingFloatChannel, keys [[0,0],[4,45]]}
          Scale.X {type=MovieSceneScriptingFloatChannel, keys [[0,1],[4,0.75]]}
          Scale.Y {type=MovieSceneScriptingFloatChannel, keys [[0,1],[4,0.75]]}
         ]
        }]
      }]
    }
    AnimBoundSizeControl {
     id=#23
     possessed_class=Border
     tracks [{
       type=MovieScene2DTransformTrack
       sections [{
         type=MovieScene2DTransformSection
         channels [
          Scale.X {type=MovieSceneScriptingFloatChannel, const [1,1]}
          Scale.Y {type=MovieSceneScriptingFloatChannel, const [1,1]}
         ]
        }]
      }]
    }
    Reticle_Spinners_AttachedFast {
     id=#24
     possessed_class=Image
     tracks [{
       type=MovieSceneFloatTrack
       display_name=Render Opacity
       sections [{type=MovieSceneFloatSection, channels [None {type=MovieSceneScriptingFloatChannel, keys [[0,0],[4,1]]}]}]
      }]
    }
    Reticle_Surrounding_Shadow {
     id=#25
     possessed_class=Image
     tracks [{
       type=MovieSceneWidgetMaterialTrack
       sections [{type=MovieSceneParameterSection, channels [LockedOn {type=MovieSceneScriptingFloatChannel, const [1,1]}]}]
      }]
    }
   ]}
 }
 CantUseNew {
  binding_count=6
  length_seconds=0.1
  object=&self:CantUseNew
  sequence {bindings [
    Reticle_Surrounding {
     id=#15
     possessed_class=Image
     tracks [{
       type=MovieSceneWidgetMaterialTrack
       sections [{
         type=MovieSceneParameterSection
         channels [
          LockedOn {type=MovieSceneScriptingFloatChannel, keys [[0,1],[2,0]]}
          AnimateAlpha {type=MovieSceneScriptingFloatChannel, const [1,1]}
         ]
        }]
      }]
    }
    Reticle_Surrounding_Shadow {
     id=#16
     possessed_class=Image
     tracks [{
       type=MovieSceneWidgetMaterialTrack
       sections [{type=MovieSceneParameterSection, channels [LockedOn {type=MovieSceneScriptingFloatChannel, keys [[0,1],[2,0]]}]}]
      }]
    }
    Reticle_Surrounding_BaseDarken {
     id=#17
     possessed_class=Image
     tracks [{
       type=MovieSceneWidgetMaterialTrack
       sections [{type=MovieSceneParameterSection, channels [LockedOn {type=MovieSceneScriptingFloatChannel, keys [[0,1],[2,0]]}]}]
      }]
    }
    Reticle_Spinners_AttachedFast {
     id=#18
     possessed_class=Image
     tracks [{
       type=MovieSceneFloatTrack
       display_name=Render Opacity
       sections [{type=MovieSceneFloatSection, channels [None {type=MovieSceneScriptingFloatChannel}]}]
      }]
    }
    AnimBoundSizeControl {
     id=#19
     possessed_class=Border
     tracks [{
       type=MovieScene2DTransformTrack
       sections [{
         type=MovieScene2DTransformSection
         channels [
          Scale.X {type=MovieSceneScriptingFloatChannel, const [1,2]}
          Scale.Y {type=MovieSceneScriptingFloatChannel, const [1,1]}
         ]
        }]
      }]
    }
    Bracket {
     id=#20
     possessed_class=Overlay
     tracks [{
       type=MovieScene2DTransformTrack
       sections [{
         type=MovieScene2DTransformSection
         channels [
          Angle {type=MovieSceneScriptingFloatChannel, const [0,1]}
          Scale.X {type=MovieSceneScriptingFloatChannel, const [1,1]}
          Scale.Y {type=MovieSceneScriptingFloatChannel, const [1,1]}
         ]
        }]
      }]
    }
   ]}
 }
 Intro {
  binding_count=1
  length_seconds=0.3
  object=&self:Intro
  sequence {bindings [CanvasPanel {
     id=#26
     possessed_class=Overlay
     tracks [
      {
       type=MovieSceneFloatTrack
       display_name=Render Opacity
       sections [{type=MovieSceneFloatSection, channels [None {type=MovieSceneScriptingFloatChannel, keys [[0,0],[6,1]]}]}]
      }
      {
       type=MovieScene2DTransformTrack
       sections [{
         type=MovieScene2DTransformSection
         channels [
          Translation.X {type=MovieSceneScriptingFloatChannel, const [0,1]}
          Translation.Y {type=MovieSceneScriptingFloatChannel, const [0,1]}
          Angle {type=MovieSceneScriptingFloatChannel, const [0,1]}
          Scale.X {type=MovieSceneScriptingFloatChannel, keys [[0,1.25],[6,1]]}
          Scale.Y {type=MovieSceneScriptingFloatChannel, keys [[0,1.25],[6,1]]}
          Shear.X {type=MovieSceneScriptingFloatChannel, const [0,1]}
          Shear.Y {type=MovieSceneScriptingFloatChannel, const [0,1]}
         ]
        }]
      }
     ]
    }]}
 }
 SwingingNew_NoLock {
  binding_count=6
  length_seconds=0.2
  object=&self:SwingingNew_NoLock
  sequence {bindings [
    Reticle_Surrounding {
     id=#15
     possessed_class=Image
     tracks [{
       type=MovieSceneWidgetMaterialTrack
       sections [{
         type=MovieSceneParameterSection
         channels [
          LockedOn {type=MovieSceneScriptingFloatChannel, keys [[0,0],[4,1]]}
          AnimateAlpha {type=MovieSceneScriptingFloatChannel, keys [[0,0],[4,1]]}
         ]
        }]
      }]
    }
    Reticle_Surrounding_BaseDarken {
     id=#21
     possessed_class=Image
     tracks [{
       type=MovieSceneWidgetMaterialTrack
       sections [{type=MovieSceneParameterSection, channels [LockedOn {type=MovieSceneScriptingFloatChannel, const [1,1]}]}]
      }]
    }
    Bracket {
     id=#22
     possessed_class=Overlay
     tracks [{
       type=MovieScene2DTransformTrack
       sections [{
         type=MovieScene2DTransformSection
         channels [
          Angle {type=MovieSceneScriptingFloatChannel, keys [[0,0],[4,45]]}
          Scale.X {type=MovieSceneScriptingFloatChannel, keys [[0,1],[4,0.75]]}
          Scale.Y {type=MovieSceneScriptingFloatChannel, keys [[0,1],[4,0.75]]}
         ]
        }]
      }]
    }
    AnimBoundSizeControl {
     id=#23
     possessed_class=Border
     tracks [{
       type=MovieScene2DTransformTrack
       sections [{
         type=MovieScene2DTransformSection
         channels [
          Scale.X {type=MovieSceneScriptingFloatChannel, const [1.1875,1]}
          Scale.Y {type=MovieSceneScriptingFloatChannel, const [1.1875,1]}
         ]
        }]
      }]
    }
    Reticle_Spinners_AttachedFast {
     id=#24
     possessed_class=Image
     tracks [{
       type=MovieSceneFloatTrack
       display_name=Render Opacity
       sections [{type=MovieSceneFloatSection, channels [None {type=MovieSceneScriptingFloatChannel, keys [[0,0],[4,1]]}]}]
      }]
    }
    Reticle_Surrounding_Shadow {
     id=#25
     possessed_class=Image
     tracks [{
       type=MovieSceneWidgetMaterialTrack
       sections [{type=MovieSceneParameterSection, channels [LockedOn {type=MovieSceneScriptingFloatChannel, const [1,1]}]}]
      }]
    }
   ]}
 }
]
graphs [
 EventGraph {nodes [
   n1 = WeaponEquipped!() -> (OutputDelegate, then, EquippedMountedWeapon:Fort Weapon Ranged Object Reference) {is=title K2Node_CustomEvent}
   > BindToWeapon!(execute, self, Weapon:Fort Weapon Ranged Object Reference=@n1.EquippedMountedWeapon) -> (then)
   n2 = BindToWeapon!() -> (OutputDelegate, then, Weapon:Fort Weapon Ranged Object Reference) {is=title K2Node_CustomEvent}
   > n3 = Cast To FortWeaponRangedOctopusTowhook!(execute, Object=@n2.Weapon) -> (then, CastFailed, AsFort Weapon Ranged Octopus Towhook:Fort Weapon Ranged Octopus Towhook Object Reference) {is=title K2Node_DynamicCast}
   > n4 = Set TowhookWeapon(TowhookWeapon:Fort Weapon Ranged Octopus Towhook Object Reference=@n3) -> (Output_Get:Fort Weapon Ranged Octopus Towhook Object Reference)
   n5 = Create Event()
   > Bind Event to On Target Updated(self:Fort Weapon Ranged Octopus Towhook Object Reference=@n4, Delegate=@n5) {is=title K2Node_AddDelegate}
   UnbindFromWeapon() {is=title K2Node_CustomEvent}
   > n6 = Get!(execute:Exec) -> (then:Exec, else:Exec, TowhookWeapon:Fort Weapon Ranged Octopus Towhook Object Reference)
   > Unbind Event from On Target Updated(self:Fort Weapon Ranged Octopus Towhook Object Reference=@n6, Delegate=@n5) {is=title K2Node_RemoveDelegate}
   n7 = OnOutOfFuelDelegate_Event() {is=title K2Node_CustomEvent}
   > Set OutOfFuel(OutOfFuel:bool=true) -> (Output_Get:bool)
   > SetColorAndOpacity(
    self:User Widget Object Reference=Get TargetWidget() -> (TargetWidget:RE West Sausage Targeting Object Reference)
    InColorAndOpacity:Linear Color Structure=Get OutOfFuelTargetWidgetColor() -> (OutOfFuelTargetWidgetColor:Linear Color Structure)
   )
   > SetColorAndOpacity(
    self:Image Object Reference=[
     Get Reticle_Surrounding() -> (Reticle_Surrounding:Image Object Reference)
     Get CooldownRing() -> (CooldownRing:Image Object Reference)
     Get Reticle_Spinners_AttachedFast() -> (Reticle_Spinners_AttachedFast:Image Object Reference)
     Get Reticle_Surrounding_BaseDarken() -> (Reticle_Surrounding_BaseDarken:Image Object Reference)
     Get Reticle_Surrounding_Disabled() -> (Reticle_Surrounding_Disabled:Image Object Reference)
    ]
    InColorAndOpacity=Get OutOfFuelReticleColor() -> (OutOfFuelReticleColor:Linear Color Structure)
   )
   CooldownTick() {is=title K2Node_CustomEvent}
   > n8 = Set CurrentCooldownDuration(CurrentCooldownDuration:double=float - float(A=GetServerWorldTimeSeconds!() -> (ReturnValue:float), B=Get CooldownStartTime() -> (CooldownStartTime:double))) -> (Output_Get:double)
   > n9 = Branch(Condition="float >= float"(A=@n8, B=Get TotalCooldownDuration() -> (TotalCooldownDuration:double)))
   n10 = Get FullCooldown() -> (FullCooldown:double)
   n11 = Get TotalCooldownDuration()
   n12 = "Max (Float)"(A=float - float(A=@n10, B=@n11), B=0)
   n13 = Get CurrentCooldownDuration() -> (CurrentCooldownDuration:double)
   SetScalarParameterValue(
    execute=@n9.else
    self=Get CooldownRingDynamicMaterial() -> (CooldownRingDynamicMaterial:Material Instance Dynamic Object Reference)
    ParameterName=Get Cooldown Percent Parameter Name() -> (Cooldown Percent Parameter Name:Name=None)
    Value=SafeDivide(A=float + float(A=@n13, B=@n12), B="Max (Float)"(A=@n10, B=@n11))
   )
   n14 = Update!() -> (OutputDelegate, then, Position:"Vector (by ref)", AbleToUse:bool=~) {is=title K2Node_CustomEvent}
   > n15 = Set TargetPosition(TargetPosition:Vector=@n14.Position) -> (Output_Get:Vector)
   > SetPosition!(
    execute
    self:Canvas Panel Slot Object Reference=SlotAsCanvasSlot!(Widget:Widget Object Reference=Get TargetWidget()) -> (ReturnValue:Canvas Panel Slot Object Reference)
    InPosition:Vector 2D Structure=WorldToScreen!(WorldLocation:Vector=@n15) -> (ScreenLocation:Vector 2D Structure)
   ) -> (then)
   > n16 = Set AbleToUse(AbleToUse:bool=@n14.AbleToUse) -> (Output_Get:bool)
   > n17 = Set OnCooldown(OnCooldown:bool=NOT Boolean(A=@n16)) -> (Output_Get:bool)
   > Branch(Condition=@n17)
   > CooldownTick() {is=title K2Node_CallFunction}
   n18 = SetLockedOnState!() -> (OutputDelegate, then, LockedOn:bool=~, Attached:bool=~) {is=title K2Node_CustomEvent}
   > Set Locked On(Locked On:bool=@n18.LockedOn) -> (Output_Get:bool)
   > Set Attached(Attached:bool=@n18.Attached) -> (Output_Get:bool)
   > n19 = Branch(Condition=Get AbleToUse() -> (AbleToUse:bool))
   n20 = Get Bracket() -> (Bracket:Overlay Object Reference)
   > SetVisibility(self=@n20, InVisibility=SelfHitTestInvisible)
   n21 = Get CooldownRing()
   > SetVisibility(self=@n21, InVisibility=Collapsed)
   > n22 = Branch(Condition=Get IsCooldownPlaying() -> (IsCooldownPlaying:bool))
   > n23 = PlayAnimation(InAnimation=Get Intro())
   Set IsCooldownPlaying(execute=[@n23.then, @n22.else], IsCooldownPlaying:bool=false) -> (Output_Get:bool)
   n24 = Get OutOfFuel() -> (OutOfFuel:bool)
   > n25 = Branch(Condition=AND Boolean(A=Get Locked On() -> (Locked On:bool), B=NOT Boolean(A=@n24)))
   > n26 = Branch(Condition=Get HasPlayedLockedOnAnim() -> (HasPlayedLockedOnAnim:bool))
   n27 = Branch(execute=[@n29.then, @n26.then], Condition=Get Attached() -> (Attached:bool))
   > SetVisibility(self=Get Reticle_Spinners_AttachedFast(), InVisibility=SelfHitTestInvisible)
   > n28 = Branch(Condition=Get HasPlayedAttachedAnim() -> (HasPlayedAttachedAnim:bool))
   PlayAnimationForward(execute=@n28.else, InAnimation=Get SwingingNew() -> (SwingingNew:Widget Animation Object Reference))
   > Set HasPlayedAttachedAnim(HasPlayedAttachedAnim:bool=true) -> (Output_Get:bool)
   SetVisibility(execute=@n27.else, self=Get Reticle_Spinners_AttachedFast(), InVisibility=Collapsed)
   > Branch(Condition=Get HasPlayedAttachedAnim())
   > PlayAnimationReverse(InAnimation=Get SwingingNew())
   > Set HasPlayedAttachedAnim(HasPlayedAttachedAnim=false)
   PlayAnimationForward(execute=@n26.else, InAnimation=Get LockedOnNew() -> (LockedOnNew:Widget Animation Object Reference))
   > n29 = Set HasPlayedLockedOnAnim(HasPlayedLockedOnAnim:bool=true) -> (Output_Get:bool)
   n30 = Branch(execute=@n25.else, Condition=Get Attached())
   SetVisibility(execute=@n30.else, self=Get Reticle_Spinners_AttachedFast(), InVisibility=Collapsed)
   > Branch(Condition=Get HasPlayedLockedOnAnim())
   > PlayAnimationReverse(InAnimation=Get LockedOnNew())
   > Set HasPlayedLockedOnAnim(HasPlayedLockedOnAnim=false)
   SetVisibility(execute=@n19.else, self=@n20, InVisibility=Collapsed)
   > SetVisibility(self=@n21, InVisibility=Select(Option 0=Collapsed, Option 1=SelfHitTestInvisible, Index=Get OnCooldown() -> (OnCooldown:bool)))
   > SetVisibility(self=Get Reticle_Spinners_AttachedFast(), InVisibility=Collapsed)
   > n31 = Branch(Condition=Get IsCooldownPlaying())
   Set IsCooldownPlaying(execute=@n31.else, IsCooldownPlaying=true)
   > PlayAnimation(InAnimation=Get Cooldown_Intro() -> (Cooldown_Intro:Widget Animation Object Reference))
   n32 = Get LockedOnNew()
   > PlayAnimation(InAnimation=@n32, StartAtTime=GetEndTime(self=@n32), PlayMode=Reverse)
   TowhookAttached() {is=title K2Node_CustomEvent}
   n33 = Get Locked On()
   > n34 = SetLockedOnState!(execute, self, LockedOn:bool=@n33, Attached:bool=true) -> (then) {is=title K2Node_CallFunction}
   Update Lock-on Visibility(execute=[@n34.then, @n35.then])
   TowhookDetached() {is=title K2Node_CustomEvent}
   > n35 = SetLockedOnState(LockedOn=@n33, Attached=false)
   n36 = TargetUpdated!() -> (OutputDelegate, then, TargetLocation:"Vector (by ref)") {is=title K2Node_CustomEvent}
   > Update!(execute, self, Position:"Vector (by ref)"=@n36.TargetLocation, AbleToUse:bool=true) -> (then) {is=title K2Node_CallFunction}
   n37 = NOT Boolean(A=Vector_IsZero!(A:"Vector (by ref)"=@n36.TargetLocation) -> (ReturnValue:bool))
   > Branch(Condition="Not Equal (Boolean)"(A=@n37, B=Get Locked On()) -> (ReturnValue=~))
   > SetLockedOnState(LockedOn=@n37, Attached=Get Attached())
   > Update Lock-on Visibility()
   n38 = Event OnVehicleUpdated!() -> (OutputDelegate, then, NewVehicle:Fort Athena Vehicle Object Reference) {is=title K2Node_Event}
   n39 = Get DisabledTag() -> (DisabledTag:Gameplay Tag Structure)
   > n40 = WaitGameplayTagAddToActor!(execute:Exec, TargetActor:Actor Object Reference=@n38.NewVehicle, Tag:Gameplay Tag Structure=@n39, OnlyTriggerOnce:bool=false) -> (then:Exec, AsyncTaskProxy:Ability Async Wait Gameplay Tag Added Object Reference, Added:Exec) {is=title K2Node_AsyncAction}
   > n41 = WaitGameplayTagRemoveFromActor!(execute:Exec, TargetActor:Actor Object Reference=@n38.NewVehicle, Tag:Gameplay Tag Structure=@n39, OnlyTriggerOnce:bool=false) -> (then:Exec, AsyncTaskProxy:Ability Async Wait Gameplay Tag Removed Object Reference, Removed:Exec) {is=title K2Node_AsyncAction}
   OnVehicleEnabled!(execute=@n41.Removed, self, Enabled:bool=true) -> (then)
   OnVehicleEnabled(execute=@n40.Added, Enabled=false)
   Event Construct()
   > Parent: Construct()
   > SetVisibility(self=Get Bracket(), InVisibility=Collapsed)
   > SetVisibility(self=Get CooldownRing(), InVisibility=Collapsed)
   > n42 = GetDynamicMaterial(self=Get CooldownRing())
   > Set CooldownRingDynamicMaterial(CooldownRingDynamicMaterial:Material Instance Dynamic Object Reference=@n42) -> (Output_Get:Material Instance Dynamic Object Reference)
   > UpdateChargeCircle!(execute, self:RE West Sausage Targeting Object Reference=Get TargetWidget(), Scale X:double=0) -> (then)
   n43 = Event BindTo!() -> (OutputDelegate, then, AssociatedObject:Object Reference) {is=title K2Node_Event}
   > Parent: BindTo!(execute, AssociatedObject:Object Reference=@n43.AssociatedObject) -> (then) {is=title K2Node_CallParentFunction}
   > n44 = Cast To FortOctopusVehicle!(execute, Object=@n43.AssociatedObject) -> (then, CastFailed, AsFort Octopus Vehicle:Fort Octopus Vehicle Object Reference) {is=title K2Node_DynamicCast}
   > n45 = Set Baller(Baller:Fort Octopus Vehicle Object Reference=@n44) -> (Output_Get:Fort Octopus Vehicle Object Reference)
   > Assign On Out Of Fuel Delegate!(execute:Exec, self:Fort Athena Vehicle Object Reference=@n45, Delegate:"Delegate (by ref)"=@n7) -> (then:Exec) {is=title K2Node_AssignDelegate}
   n46 = GetSeatWeaponComponent!(self:Fort Vehicle Interface Interface=@n45, SeatIndex:int=0) -> (ReturnValue:Fort Vehicle Seat Weapon Component Object Reference)
   > Is Valid(InputObject=@n46)
   > n47 = Get!(execute, self:Fort Vehicle Seat Weapon Component Object Reference=@n46) -> (then, else, CachedWeapon:Fort Weapon Ranged Weak object)
   > n48 = BindToWeapon(Weapon=@n47) {is=title K2Node_CallFunction}
   n49 = Get Baller() -> (Baller:Fort Octopus Vehicle Object Reference)
   Bind Event to On Towhook Attached(execute=[@n50.then, @n48.then], self:Fort Octopus Vehicle Object Reference=@n49, Delegate=Create Event()) {is=title K2Node_AddDelegate}
   > Bind Event to On Towhook Detached(self:Fort Octopus Vehicle Object Reference=@n49, Delegate=Create Event()) {is=title K2Node_AddDelegate}
   n50 = Bind Event to On Equip Mounted Weapon(execute=@n47.else, self:Fort Vehicle Seat Weapon Component Object Reference=@n46, Delegate=@n1) {is=title K2Node_AddDelegate}
   n51 = Event UnbindFrom!() -> (OutputDelegate, then, AssociatedObject:Object Reference) {is=title K2Node_Event}
   > Parent: UnbindFrom!(execute, AssociatedObject:Object Reference=@n51.AssociatedObject) -> (then) {is=title K2Node_CallParentFunction}
   > UnbindFromWeapon() {is=title K2Node_CallFunction}
   > n52 = Get!(execute) -> (then, else, Baller:Fort Octopus Vehicle Object Reference)
   n53 = GetSeatWeaponComponent(self=@n52, SeatIndex=0)
   > Is Valid(InputObject=@n53)
   > Unbind Event from On Equip Mounted Weapon(self:Fort Vehicle Seat Weapon Component Object Reference=@n53, Delegate=@n1) {is=title K2Node_RemoveDelegate}
  ]}
 GetTargetSize {nodes [
   n1 = Inputs!() -> (execute, PhysicsObject:Actor Object Reference)
   > n2 = GetPhysicsObjectSimSize!(execute, PhysicsObject:Actor Object Reference=@n1) -> (then, ReturnValue:EFortPhysicsSimSize Enum=Small)
   n3 = Cast To FortPickup!(Object=@n1) -> (AsFort Pickup:Fort Pickup Object Reference, bSuccess:bool=~) {is=title K2Node_DynamicCast}
   n4 = Cast To FortProjectileBase!(Object=@n1) -> (AsFort Projectile Base:Fort Projectile Base Object Reference, bSuccess:bool=~) {is=title K2Node_DynamicCast}
   > Outputs!(
    then:Exec
    ReturnValue:EFortPhysicsSimSize Enum=Select(Option 0:EFortPhysicsSimSize Enum=@n2, Option 1:EFortPhysicsSimSize Enum=Small, Index=OR Boolean(A=@n3.bSuccess, B=@n4.bSuccess)) -> (ReturnValue:EFortPhysicsSimSize Enum)
   )
  ]}
 OnVehicleEnabled {nodes [
   n1 = OnVehicleEnabled!() -> (then, Enabled:bool=~) {is=title K2Node_FunctionEntry}
   > n2 = Branch(Condition=@n1)
   n3 = Get WBP_Vehicle_Warning() -> (WBP_Vehicle_Warning:WBP Vehicle Warning Object Reference)
   > RemoveWarning!(execute, self:WBP Vehicle Warning Object Reference=@n3, Type:ENUM_WarningType Enum=NewEnumerator2) -> (then)
   > SetVisibility(self=@n3, InVisibility=Collapsed)
   ShowWarning!(execute=@n2.else, self:WBP Vehicle Warning Object Reference=@n3, Type:ENUM_WarningType Enum=NewEnumerator2) -> (then)
   > SetVisibility(self=@n3, InVisibility=HitTestInvisible)
  ]}
 Update Lock-on Visibility {nodes [
   Update Lock-on Visibility() {is=title K2Node_FunctionEntry}
   n1 = Get Attached()
   > n2 = Branch(Condition=AND Boolean!(A=Get AbleToUse(), B=Get Locked On(), C:bool=NOT Boolean(A=@n1)) -> (ReturnValue))
   n3 = Get TargetWidget()
   > SetVisibility(self=@n3, InVisibility=HitTestInvisible)
   SetVisibility(execute=@n2.else, self=@n3, InVisibility=Collapsed)
  ]}
 WorldToScreen {nodes [
   n1 = WorldToScreen!() -> (then, WorldLocation:Vector=~) {is=title K2Node_FunctionEntry}
   > SetWidgetCZonesAutomation!(execute, bEnabled:bool=true) -> (then)
   > n2 = Set TmpPos(TmpPos:Vector 2D Structure=ProjectToWidgetPosition!(PlayerController:Player Controller Object Reference=GetOwningPlayer(), WorldLocation:Vector=@n1, bPlayerViewportRelative:bool=false) -> (ScreenPosition:Vector 2D Structure, ReturnValue:bool)) -> (Output_Get:Vector 2D Structure)
   > SetWidgetCZonesAutomation(bEnabled=false)
   n3 = BreakVector2D(InVec=GetViewportSize!() -> (ReturnValue:Vector 2D Structure))
   n4 = float / float(A=@n3, B=2)
   n5 = float / float(A=@n3.Y, B=2)
   > Return Node!(
    execute
    ScreenLocation:Vector 2D Structure=vector2d - vector2d(A:Vector 2D Structure=@n2, B:Vector 2D Structure=MakeVector2D(X=@n4, Y=@n5)) -> (ReturnValue:Vector 2D Structure) {is=title K2Node_PromotableOperator}
   )
  ]}
]
variables [
 AbleToUse {type=boolean}
 Attached {type=boolean}
 Baller {type=FortOctopusVehicle, default=null}
 Cooldown Percent Parameter Name {type=string, default=CooldownPercent}
 Cooldown_GE {type=ActiveGameplayEffectHandle, default {bPassedFiltersAndWasExecuted=false}}
 CooldownAnimPlayed {type=boolean}
 CooldownInterval {type=number}
 CooldownRingDynamicMaterial {type=MaterialInstanceDynamic}
 CooldownStartTime {type=number}
 CooldownTime {type=number}
 CooldownTimerHandle {type=TimerHandle}
 CurrentCooldownDuration {type=number}
 DisabledTag {type=GameplayTag, default {tagName=Vehicle.Disabled}}
 FullCooldown {type=number, default=8}
 FullCooldownTime {type=ScalableFloat, default {registryType=None, value=0, curve {curveTable=null, rowName=None}}}
 HasPlayedAttachedAnim {type=boolean}
 HasPlayedCantUseAnim {type=boolean}
 HasPlayedLockedOnAnim {type=boolean}
 IsCooldownPlaying {type=boolean}
 Locked On {type=boolean}
 OnCooldown {type=boolean}
 OutOfFuel {type=boolean}
 OutOfFuelReticleColor {type=LinearColor, default=<rgba 1 0 0 0.5>}
 OutOfFuelTargetWidgetColor {type=LinearColor, default=<rgba 0 0 0 0>}
 Player {type=Actor, default=null}
 TargetPosition {type=Vector}
 TotalCooldownDuration {type=number}
 TowhookWeapon {type=FortWeaponRangedOctopusTowhook, default=null}
]
widgets {
 info {
  inheritedWidgetCount=10
  parentClass=/Script/FortniteUI.FortAthenaVehicleDashboardWidget
  rootWidgetClass=&Overlay
  widgetCount=22
 }
 widgets [
  Overlay_1 {widgetClassPath=&Overlay, widget=Overlay_1}
  CanvasPanel {
   widgetClassPath=&CanvasPanel
   parent=Overlay_1
   widget=CanvasPanel
   slot {class=OverlaySlot, properties {horizontalAlignment=HAlign_Center, verticalAlignment=VAlign_Center}}
  }
  TargetWidget {
   widgetClassPath=/GrappleParent/GrappleParent/Reticle/RE_WestSausage_Targeting
   bIsVariable=true
   parent=CanvasPanel
   widget=TargetWidget
   slot {
    class=CanvasPanelSlot
    properties {layoutData {alignment=<0.5 0.5>, anchors {maximum=<0.5 0.5>, minimum=<0.5 0.5>}, offsets=<ltrb 0 0 30 30>}}
   }
  }
  AnimBoundSizeControl {
   widgetClassPath=&Border
   parent=Overlay_1
   widget=AnimBoundSizeControl
   properties {
    visibility=SelfHitTestInvisible
    background {
     bIsDynamicallyLoaded=false
     drawAs=NoDrawType
     imageType=NoImage
     mirroring=NoMirror
     resourceName=None
     resourceObject=null
     tiling=NoTile
     imageSize=<32 32>
     margin=<ltrb 0 0 0 0>
     outlineSettings {
      bUseBrushTransparency=false
      roundingType=HalfHeightRadius
      width=0
      color {colorUseRule=UseColor_Specified, specifiedColor=<rgba 0 0 0 0>}
      cornerRadii=<0 0 0 0>
     }
     tintColor {colorUseRule=UseColor_Specified, specifiedColor=<rgba 1 1 1 1>}
     uVRegion {bIsValid=false, max=<0 0>, min=<0 0>}
    }
    padding=<ltrb 0 0 0 0>
    renderTransform {angle=0, scale=<1.1875 1.1875>, shear=<0 0>, translation=<0 0>}
   }
   slot {class=OverlaySlot, properties {horizontalAlignment=HAlign_Center, verticalAlignment=VAlign_Center}}
  }
  Bracket {
   widgetClassPath=&Overlay
   bIsVariable=true
   parent=AnimBoundSizeControl
   widget=Bracket
   slot {class=BorderSlot, properties {padding=<ltrb 0 0 0 0>}}
  }
  Reticle_Surrounding_Shadow {
   widgetClassPath=&Image
   bIsVariable=true
   parent=Bracket
   widget=Reticle_Surrounding_Shadow
   properties {
    visibility=SelfHitTestInvisible
    brush {
     bIsDynamicallyLoaded=false
     drawAs=Image
     imageType=NoImage
     mirroring=NoMirror
     resourceName=None
     resourceObject=/GrappleParent/GrappleParent/Reticle/M_UI_WestSausage_ReticleBracket_Shadow
     tiling=NoTile
     imageSize=<96 96>
     margin=<ltrb 0 0 0 0>
     outlineSettings {
      bUseBrushTransparency=false
      roundingType=HalfHeightRadius
      width=0
      color {colorUseRule=UseColor_Specified, specifiedColor=<rgba 0 0 0 0>}
      cornerRadii=<0 0 0 0>
     }
     tintColor {colorUseRule=UseColor_Specified, specifiedColor=<rgba 0 0 0 1>}
     uVRegion {bIsValid=false, max=<0 0>, min=<0 0>}
    }
   }
   slot {class=OverlaySlot, properties {horizontalAlignment=HAlign_Center, verticalAlignment=VAlign_Center}}
  }
  Reticle_Surrounding_BaseDarken {
   widgetClassPath=&Image
   bIsVariable=true
   parent=Bracket
   widget=Reticle_Surrounding_BaseDarken
   properties {
    visibility=SelfHitTestInvisible
    brush {
     bIsDynamicallyLoaded=false
     drawAs=Image
     imageType=NoImage
     mirroring=NoMirror
     resourceName=None
     resourceObject=/GrappleParent/GrappleParent/Reticle/M_UI_WestSausage_ReticleBracket_BaseDarken
     tiling=NoTile
     imageSize=<96 96>
     margin=<ltrb 0 0 0 0>
     outlineSettings {
      bUseBrushTransparency=false
      roundingType=HalfHeightRadius
      width=0
      color {colorUseRule=UseColor_Specified, specifiedColor=<rgba 0 0 0 0>}
      cornerRadii=<0 0 0 0>
     }
     tintColor {colorUseRule=UseColor_Specified, specifiedColor=<rgba 1 1 1 1>}
     uVRegion {bIsValid=false, max=<0 0>, min=<0 0>}
    }
   }
   slot {class=OverlaySlot}
  }
  Reticle_Surrounding_Disabled {
   widgetClassPath=&Image
   bIsVariable=true
   parent=Bracket
   widget=Reticle_Surrounding_Disabled
   properties {
    visibility=Hidden
    brush {
     bIsDynamicallyLoaded=false
     drawAs=Image
     imageType=NoImage
     mirroring=NoMirror
     resourceName=None
     resourceObject=/GrappleParent/GrappleParent/Reticle/M_UI_WestSausage_ReticleBracket_Disabled
     tiling=NoTile
     imageSize=<96 96>
     margin=<ltrb 0 0 0 0>
     outlineSettings {
      bUseBrushTransparency=false
      roundingType=HalfHeightRadius
      width=0
      color {colorUseRule=UseColor_Specified, specifiedColor=<rgba 0 0 0 0>}
      cornerRadii=<0 0 0 0>
     }
     tintColor {colorUseRule=UseColor_Specified, specifiedColor=<rgba 1 1 1 1>}
     uVRegion {bIsValid=false, max=<0 0>, min=<0 0>}
    }
   }
   slot {class=OverlaySlot, properties {horizontalAlignment=HAlign_Fill, verticalAlignment=VAlign_Fill}}
  }
  Reticle_Surrounding {
   widgetClassPath=&Image
   bIsVariable=true
   parent=Bracket
   widget=Reticle_Surrounding
   properties {
    visibility=SelfHitTestInvisible
    brush {
     bIsDynamicallyLoaded=false
     drawAs=Image
     imageType=NoImage
     mirroring=NoMirror
     resourceName=None
     resourceObject=/GrappleParent/GrappleParent/Reticle/M_UI_WestSausage_ReticleBracket
     tiling=NoTile
     imageSize=<96 96>
     margin=<ltrb 0 0 0 0>
     outlineSettings {
      bUseBrushTransparency=false
      roundingType=HalfHeightRadius
      width=0
      color {colorUseRule=UseColor_Specified, specifiedColor=<rgba 0 0 0 0>}
      cornerRadii=<0 0 0 0>
     }
     tintColor {colorUseRule=UseColor_Specified, specifiedColor=<rgba 1 1 1 1>}
     uVRegion {bIsValid=false, max=<0 0>, min=<0 0>}
    }
   }
   slot {class=OverlaySlot}
  }
  Reticle_Spinners_AttachedFast {
   widgetClassPath=&Image
   bIsVariable=true
   parent=Bracket
   widget=Reticle_Spinners_AttachedFast
   properties {
    renderOpacity=0
    visibility=Hidden
    brush {
     bIsDynamicallyLoaded=false
     drawAs=Image
     imageType=NoImage
     mirroring=NoMirror
     resourceName=None
     resourceObject=/GrappleParent/GrappleParent/Reticle/M_UI_WestSausage_Reticle_BracketSpinningCircles_Fast
     tiling=NoTile
     imageSize=<56 56>
     margin=<ltrb 0 0 0 0>
     outlineSettings {
      bUseBrushTransparency=false
      roundingType=HalfHeightRadius
      width=0
      color {colorUseRule=UseColor_Specified, specifiedColor=<rgba 0 0 0 0>}
      cornerRadii=<0 0 0 0>
     }
     tintColor {colorUseRule=UseColor_Specified, specifiedColor=<rgba 1 1 1 1>}
     uVRegion {bIsValid=false, max=<0 0>, min=<0 0>}
    }
   }
   slot {class=OverlaySlot, properties {horizontalAlignment=HAlign_Center, verticalAlignment=VAlign_Center}}
  }
  CooldownRing {
   widgetClassPath=&Image
   bIsVariable=true
   parent=Overlay_1
   widget=CooldownRing
   properties {
    visibility=HitTestInvisible
    brush {
     bIsDynamicallyLoaded=false
     drawAs=Image
     imageType=NoImage
     mirroring=NoMirror
     resourceName=None
     resourceObject=/GrappleParent/GrappleParent/Reticle/M_UI_WestSausage_Reticle_Ring
     tiling=NoTile
     imageSize=<56 56>
     margin=<ltrb 0 0 0 0>
     outlineSettings {
      bUseBrushTransparency=false
      roundingType=HalfHeightRadius
      width=0
      color {colorUseRule=UseColor_Specified, specifiedColor=<rgba 0 0 0 0>}
      cornerRadii=<0 0 0 0>
     }
     tintColor {colorUseRule=UseColor_Specified, specifiedColor=<rgba 1 1 1 1>}
     uVRegion {bIsValid=false, max=<0 0>, min=<0 0>}
    }
   }
   slot {class=OverlaySlot, properties {horizontalAlignment=HAlign_Center, verticalAlignment=VAlign_Center}}
  }
  WBP_Vehicle_Warning {
   widgetClassPath=/Game/Athena/HUD/Vehicles/Dashboard/WBP_Vehicle_Warning
   bIsVariable=true
   parent=Overlay_1
   widget=WBP_Vehicle_Warning
   slot {class=OverlaySlot, properties {horizontalAlignment=HAlign_Center, verticalAlignment=VAlign_Center}}
  }
  Widget_FuelGauge {widgetClassPath=/Script/FortniteUI.FortVehicleFuelGauge, bInherited=true, widget=null}
  HorizontalBox_FuelCluster {widgetClassPath=&HorizontalBox, bInherited=true, widget=null}
  Image_Speedometer {widgetClassPath=&Image, bInherited=true, widget=null}
  Image_BoostGlow {widgetClassPath=&Image, bInherited=true, widget=null}
  Image_BoostMeter {widgetClassPath=&Image, bInherited=true, widget=null}
  Overlay_BoostSparks {widgetClassPath=&Overlay, bInherited=true, widget=null}
  NumericText_Speed {widgetClassPath=/Script/CommonUI.CommonNumericTextBlock, bInherited=true, widget=null}
  Overlay_Gauge {widgetClassPath=&Overlay, bInherited=true, widget=null}
  Text_ItemName {widgetClassPath=&CommonTextBlock, bInherited=true, widget=null}
  TutorialHighlight {widgetClassPath=&TutorialHighlight}
 ]
}
